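# Material function for the PuffinSmile character that adds stylized halftone-dot rim lighting, highlights and optional animated lens glints, switched by UsePuffinSmileShading.
# Dots tile on an actor-centered, camera-facing virtual plane, scaling with FOV and distance.
# Masks combine Fresnel, spherical-blended normals and a directional-light vector, which falls back to the collection sun vector.
# Results rewrite emissive, base color, metallic, specular and roughness; lens stripes shift with view direction per eye.
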MaterialFunction /BRCosmetics/Characters/Player/Male/Medium/Bodies/M_MED_PuffinSmile/Materials/MF_PuffinSmile_DotRimLight
graph {nodes [
  ComponentMask(None=ViewProperty() {property=MEVP_TanHalfFieldOfView}.Property) {desc=Horizontal FOV, r=true}
  n1 = GetMaterialAttributes(None=NamedRerouteDeclaration(None=FunctionInput() {inputName=MA_In, inputType=FunctionInput_MaterialAttributes}) {name=MA_IN, nodeColor=<rgba 1 0.5625 0 1>}) {attributeGetTypes [
    #1
    #2
    #3
    #4
    #5
    #6
   ]}
  n2 = NamedRerouteUsage()
  n3 = Constant()
  n4 = SetMaterialAttributes!(
   MaterialAttributes=@n1.MaterialAttributes
   EmissiveColor=NamedRerouteUsage()
   BaseColor=NamedRerouteUsage()
   Metallic=LinearInterpolate(A=@n1.Metallic, Alpha=@n2) {constB=0}
   Specular=LinearInterpolate(A=Multiply(A=@n1.Specular, B=@n3), Alpha=@n2) {constB=0}
   Roughness=LinearInterpolate(A=@n1.Roughness, Alpha=@n2)
  ) {attributeSetTypes [
    #2
    #3
    #4
    #5
    #6
   ]}
  FunctionOutput(None=StaticSwitchParameter(True=@n4, False=NamedRerouteUsage()) {group=PuffinSmile, parameterName=UsePuffinSmileShading}) {outputName=MA_Out}
  n5 = Normalize(VectorInput=LinearInterpolate(A=NamedRerouteUsage(), B=NamedRerouteUsage(), Alpha=ScalarParameter() {group=PuffinSmile_DotRimLight, parameterName=SphericalNormalBlending}) {constAlpha=1})
  n6 = Fresnel(
   ExponentIn=ScalarParameter() {defaultValue=1, group=PuffinSmile_DotRimLight, parameterName=RimLight_FresnelExponent, sliderMax=12}
   Normal=@n5
  ) {baseReflectFraction=0}
  InverseLinearInterpolate(
   A=ScalarParameter() {defaultValue=0.1, group=PuffinSmile_DotRimLight, parameterName=RimLight_Min, sliderMax=1}
   B=ScalarParameter() {defaultValue=0.2, group=PuffinSmile_DotRimLight, parameterName=RimLight_Max, sliderMax=1}
   Value=@n6
  ) {constA=0.1, constB=0.2}
  MaterialFunctionCall!(Fade Offset=Constant() {r=10}, Fade Distance=Constant() {r=400}) {materialFunction=/Game/Global/Materials/Character_Parent_2023/Anime_Parent_2023/MaterialFunctions/MFC_DistanceFade_FOV_compensation}
  NamedRerouteDeclaration(None=Transform(None=@n1.Normal)) {name=Normal_In, nodeColor=<rgba 1 0.3516 0 1>}
  n7 = NamedRerouteUsage()
  n8 = MaterialFunctionCall!(Vector=@n7) {materialFunction=/Game/Packages/Fortress_SharedMaterials/Anime_Master_Material/MF_ProjectToScreenSpace}
  NamedRerouteDeclaration(None=@n8.Result) {name=Debug, nodeColor=<rgba 0.6406 0 1 1>}
  n9 = DotProduct(
   A=Transform(None=NamedRerouteUsage()) {transformSourceType=TRANSFORMSOURCE_World, transformType=TRANSFORM_Local}
   B=Constant3Vector() {constant=<rgba 1 0 0 0>}
  )
  n10 = Multiply(A=@n9, B=@n9)
  NamedRerouteDeclaration(None=LinearInterpolate(
    B=Multiply(A=@n10, B=@n10) {desc=reduce rim light on the Y axis}
    Alpha=ScalarParameter() {group=PuffinSmile_DotRimLight, parameterName=YMask_Intensity}
   )) {name=YMask, nodeColor=<rgba 0 1 0.5078 1>}
  n11 = NamedRerouteUsage()
  n12 = ScalarParameter() {defaultValue=0.5, group=PuffinSmile_DotRimLight, parameterName=Dots_AOMaskIntensity, sortPriority=30}
  NamedRerouteDeclaration(None=Saturate(None=LinearInterpolate(B=@n11, Alpha=@n12) {constA=1})) {name=AO_Multiplier, nodeColor=<rgba 0.5937 1 0 1>}
  n13 = VectorParameter() {group=PuffinSmile, parameterName=Lenses_Color, sortPriority=50, defaultValue=<rgba 0.475 0.475 0.475 1>}
  n14 = VectorParameter() {group=PuffinSmile, parameterName=Lenses_HighlightColor, sortPriority=50, defaultValue=<rgba 0.95 0.95 0.95 1>}
  n15 = LinearInterpolate(A=@n13.RGB, B=@n14.RGB, Alpha=NamedRerouteUsage())
  n16 = NamedRerouteUsage()
  NamedRerouteDeclaration(None=StaticSwitchParameter(True=Multiply(A=@n15, B=@n16), False=Constant()) {group=PuffinSmile, parameterName=UseLenses, sortPriority=49}) {name=LensesColor, nodeColor=<rgba 0.2187 1 0 1>}
  NamedRerouteDeclaration(None=@n1.BaseColor) {name=BaseColor_In, nodeColor=<rgba 0 0.7187 1 1>}
  n17 = Multiply(
   A=Add(A=DotProduct(A=@n8.Result, B=@n5), B=ScalarParameter() {defaultValue=0.7, group=PuffinSmile_DotRimLight, parameterName=RimLight_NdotLOffset}) {constB=0.5}
   B=Saturate(None=@n6)
  )
  n18 = Abs(None=ComponentMask(None=@n7) {r=true})
  n19 = LinearInterpolate(
   B=ScalarParameter() {defaultValue=2, group=PuffinSmile_DotRimLight, parameterName=RimLight_PowerTODComp}
   Alpha=Saturate(None=Subtract(A=@n18) {constB=0.1})
  ) {constA=1, constB=5}
  n20 = ScalarParameter() {group=PuffinSmile_DotRimLight, parameterName=RimLight_IsFE}
  NamedRerouteDeclaration(None=Power(Base=Power(Base=@n17, Exp=@n19), Exp=LinearInterpolate(Alpha=@n20) {constA=1, constB=1.5})) {name=RimLight, nodeColor=<rgba 0 1 0.8359 1>}
  n21 = Multiply(
   A=Clamp(None=NamedRerouteUsage(), Max=ScalarParameter() {defaultValue=1, group=PuffinSmile_DotRimLight, parameterName=Emissive_BaseColorMax}) {maxDefault=0.5}
   B=ScalarParameter() {defaultValue=0.25, group=PuffinSmile_DotRimLight, parameterName=Emissive_Intensity, sortPriority=60}
  )
  n22 = NamedRerouteUsage()
  n23 = NamedRerouteUsage()
  NamedRerouteDeclaration(None=Add(A=LinearInterpolate(A=@n21, B=@n22, Alpha=@n23) {constB=0}, B=NamedRerouteUsage())) {name=Emissive_OUT, nodeColor=<rgba 0.4531 0 1 1>}
  n24 = NamedRerouteUsage()
  n25 = ComponentMask(None=@n24) {b=true, g=true, r=true}
  n26 = Multiply(A=ComponentMask(None=@n24) {a=true}, B=NamedRerouteUsage())
  n27 = NamedRerouteUsage()
  n28 = ComponentMask(None=@n27) {b=true, g=true, r=true}
  n29 = Multiply(A=ComponentMask(None=@n27) {a=true}, B=NamedRerouteUsage())
  NamedRerouteDeclaration(None=Add(A=Multiply(A=@n25, B=@n26) {desc=Highlight Dots}, B=Multiply(A=@n28, B=@n29))) {name=FinalColor, nodeColor=<rgba 0.7812 1 0 1>}
  NamedRerouteDeclaration(None=@n1.EmissiveColor) {name=Emissive_IN, nodeColor=<rgba 0 1 0.25 1>}
  n30 = ScalarParameter() {defaultValue=0.5, desc=0.05, group=PuffinSmile_DotRimLight, parameterName=Highlight_SmoothMax, sliderMax=1, sortPriority=35}
  n31 = ScalarParameter() {defaultValue=-0.4, desc=-0.15, group=PuffinSmile_DotRimLight, parameterName=Highlight_Offset, sliderMax=1, sliderMin=-1, sortPriority=35}
  n32 = Add(A=NamedRerouteUsage(), B=@n31)
  n33 = ScalarParameter() {defaultValue=1, desc=-0.15, group=PuffinSmile_DotRimLight, parameterName=Highlight_Transition, sliderMax=1, sliderMin=-1, sortPriority=35}
  n34 = Saturate(None=Subtract(A=Multiply(A=@n32, B=@n33), B=NamedRerouteUsage()) {constA=0.25})
  n35 = NamedRerouteDeclaration(None=Saturate(None=SmoothStep(Max=@n30, Value=@n34) {constMax=0.1})) {name=HighlightMask, nodeColor=<rgba 1 0 0.6094 1>}
  n36 = VectorParameter() {group=PuffinSmile_DotRimLight, parameterName=Dots_HighlightColor, defaultValue=<rgba 0.9 0.9 0.9 0>}
  NamedRerouteDeclaration(None=AppendVector(A=Multiply(A=@n35, B=@n36.RGB), B=@n36.A)) {name=Highlight, nodeColor=<rgba 0 1 0.9063 1>}
  n37 = Add(A=@n26, B=@n29)
  n38 = NamedRerouteUsage()
  NamedRerouteDeclaration(None=Saturate(None=LinearInterpolate(A=@n37, Alpha=@n38) {constB=0})) {name=FinalAlpha, nodeColor=<rgba 0.6172 1 0 1>}
  n39 = MaterialFunctionCall() {materialFunction=&CameraDirectionVector}
  n40 = Multiply(A=Constant3Vector() {constant=<rgba 1 1 0 0>}, B=@n39.Result)
  n41 = Add(A=Multiply(B=Normalize(VectorInput=@n40)) {constA=0.25}, B=@n39.Result)
  n42 = Multiply(A=@n39.Result) {constB=50}
  n43 = Add(A=ActorPositionWS(), B=Constant3Vector() {constant=<rgba 0 0 90 1>})
  n44 = Constant() {r=64}
  n45 = Constant() {r=128}
  n46 = InverseLinearInterpolate(
   A=Constant() {r=45}
   B=Constant() {r=90}
   Value=MaterialFunctionCall() {materialFunction=/Engine/Functions/Engine_MaterialFunctions02/Utility/FOV}.FOV angle in degrees
  )
  NamedRerouteDeclaration(None=MaterialFunctionCall(Plane Normal Axis=Normalize(VectorInput=@n41), Plane Center=Add(A=@n42, B=@n43), UV Scale=LinearInterpolate(A=@n44, B=@n45, Alpha=@n46)) {materialFunction=&VirtualPlaneCoordinates}.UVs) {name=ObjectAlighedScreenPosition, nodeColor=<rgba 1 0.867 0 1>}
  n47 = Custom() {code="return View.DirectionalLightDirection;", description=DirectionalLightVector, inputs []}
  n48 = DotProduct(A=@n47, B=@n47)
  n49 = CollectionParameter() {
   collection=&FortniteMaterialParameters
   parameterName=SunlightDirectionalVector
  }
  n50 = ComponentMask(None=@n49) {b=true, g=true, r=true}
  n51 = NamedRerouteDeclaration(None=If(
    A=Custom() {code="return View.DirectionalLightColor;", description=DirectionalLightColor, inputs []}
    "A > B"=If(A=@n48, "A > B"=@n47, "A == B"=@n50, "A < B"=@n50) {desc=check if DirectionalLight exists}
    "A == B"=@n50
    "A < B"=@n50
   ) {desc=check if DirectionalLight color is Black}) {name=LightVector, nodeColor=<rgba 1 0.5156 0 1>}
  n52 = NamedRerouteUsage()
  n53 = Multiply(A=NamedRerouteDeclaration(None=DotProduct(A=@n51, B=@n52)) {name=LdotN, nodeColor=<rgba 0.2656 0 1 1>}) {constB=0.5}
  NamedRerouteDeclaration(None=Saturate(None=Add(A=@n53) {constB=0.5})) {name=Lambert, nodeColor=<rgba 0 0.9297 1 1>}
  n54 = ActorPositionWS()
  n55 = CameraPositionWS()
  n56 = Subtract(A=Multiply(B=Distance(A=@n54, B=@n55)) {constA=1}) {constB=100}
  NamedRerouteDeclaration(None=Saturate(None=Divide(A=@n56) {constB=400})) {name=DistanceBlending, nodeColor=<rgba 0.4766 1 0 1>}
  NamedRerouteDeclaration(None=Multiply(A=NamedRerouteUsage(), B=ScalarParameter() {group=PuffinSmile_DotRimLight, parameterName=Basecolor_Intensity})) {name=BaseColor_OUT, nodeColor=<rgba 1 0.9609 0 1>}
  n57 = Normalize(VectorInput=Multiply(A=NamedRerouteUsage(), B=Constant3Vector() {constant=<rgba 1 1 0 0>}))
  n58 = Constant3Vector() {constant=<rgba 1 0 0 0>}
  NamedRerouteDeclaration(None=Multiply(A=DotProduct(A=@n57, B=@n58)) {constB=0.08}) {name=Rotation, nodeColor=<rgba 0.0312 1 0 1>}
  n59 = FunctionInput() {bUsePreviewValueAsDefault=true, inputName=M, sortPriority=1, previewValue=<0 0.5 0 1>}
  NamedRerouteDeclaration(None=ComponentMask(None=@n59) {r=true}) {name=AOMap, nodeColor=<rgba 0.7109 1 0 1>}
  n60 = ScalarParameter() {defaultValue=130, group=PuffinSmile_DotRimLight, parameterName=Dots_Amount, sliderMax=512, sortPriority=30}
  n61 = NamedRerouteUsage()
  n62 = ScalarParameter() {defaultValue=1, group=PuffinSmile_DotRimLight, parameterName=DistanceBlending, sliderMax=1, sortPriority=20}
  n63 = Multiply(A=NamedRerouteUsage(), B=LinearInterpolate(A=@n60, B=Multiply(A=@n60) {constB=0.5}, Alpha=LinearInterpolate(B=@n61, Alpha=@n62) {constA=1}))
  n64 = Constant() {r=1}
  n65 = Add(A=Frac(None=Rotator(Coordinate=@n63, Time=@n64) {speed=0.75})) {constB=-0.5}
  n66 = Constant()
  NamedRerouteDeclaration(None=Add(A=Distance(A=@n65, B=@n66)) {constB=-0.1, desc=to get full transition}) {name=DotsGradient, nodeColor=<rgba 1 0.0938 0 1>}
  n67 = NamedRerouteUsage()
  n68 = ScalarParameter() {defaultValue=0.1, group=PuffinSmile_DotRimLight, parameterName=RimDots_Power, sortPriority=30}
  n69 = ScalarParameter() {defaultValue=-0.2, group=PuffinSmile_DotRimLight, parameterName=RimDots_Offset, sliderMax=0.5, sliderMin=-0.5, sortPriority=30}
  n70 = Multiply(
   A=Add(A=Power(Base=@n67, Exp=@n68), B=@n69) {constB=-0.25}
   B=ScalarParameter() {defaultValue=1, group=PuffinSmile_DotRimLight, parameterName=RimDots_Transition, sliderMax=2, sortPriority=30}
  ) {constB=1.5}
  n71 = LinearInterpolate(A=@n70, B=Add(A=@n70) {constB=-0.25}, Alpha=NamedRerouteUsage()) {constA=1, constB=16}
  n72 = Subtract(A=@n71, B=NamedRerouteUsage()) {constA=0.25}
  n73 = ScalarParameter() {defaultValue=0.5, group=PuffinSmile_DotRimLight, parameterName=RimDots_SolidThreshold, sliderMax=1}
  n74 = Saturate(None=SmoothStep(
    Max=ScalarParameter() {defaultValue=0.35, group=PuffinSmile_DotRimLight, parameterName=RimDots_SolidlightThreshold_Max, sortPriority=30}
    Value=Subtract(A=@n71, B=@n73) {constB=0.5}
   ))
  n75 = LinearInterpolate(
   A=SmoothStep(
    Max=ScalarParameter() {defaultValue=0.4, group=PuffinSmile_DotRimLight, parameterName=RimDots_Softness, sortPriority=30}
    Value=Saturate(None=@n72)
   )
   B=@n74
   Alpha=ScalarParameter() {defaultValue=0.5, group=PuffinSmile_DotRimLight, parameterName=DotsAndSolid_Blend}
  ) {desc=Blend between dots and solid rim light}
  n76 = Multiply(A=NamedRerouteDeclaration(None=Saturate(None=@n75)) {name=RimDotsMask, nodeColor=<rgba 1 0 0.1262 1>}, B=NamedRerouteUsage())
  n77 = Desaturation(
   None=Multiply(
    A=NamedRerouteUsage()
    B=ScalarParameter() {defaultValue=3.5, group=PuffinSmile_DotRimLight, parameterName=RimDots_BaseColorMulti, sortPriority=30}
   )
   Fraction=Constant() {r=1}
  )
  n78 = VectorParameter() {group=PuffinSmile_DotRimLight, parameterName=RimDots_HighlightDotsColor, defaultValue=<rgba 0.9 0.9 0.9 1.5>}
  n79 = LinearInterpolate(
   A=MaterialFunctionCall(X=@n77, Y=@n77, Z=@n77) {desc=Remove colors from texture, materialFunction=&MakeFloat3}.Result
   B=@n78.RGB
   Alpha=@n74
  )
  NamedRerouteDeclaration(None=AppendVector(A=Multiply(A=@n76, B=@n79), B=@n78.A)) {name=RimDots, nodeColor=<rgba 0.9242 1 0.9798 1>}
  n80 = FunctionInput() {inputName=TechArtMask}
  n81 = ChannelMaskParameter(None=@n80) {group=PuffinSmile, maskChannel=Green, parameterName=Shadow_Channel, sortPriority=50, defaultValue=<rgba 0 1 0 0>}
  NamedRerouteDeclaration(None=@n81.R) {name=ShadowMask, nodeColor=<rgba 0.3359 1 0 1>}
  NamedRerouteDeclaration(None=ComponentMask(None=FunctionInput() {bUsePreviewValueAsDefault=true, inputName=STT, sortPriority=1, previewValue=<0 0.5 0 1>}) {g=true}) {name=ThresholdMap, nodeColor=<rgba 0.7109 1 0 1>}
  NamedRerouteDeclaration(None=ChannelMaskParameter(None=@n80) {group=PuffinSmile, parameterName=Lenses_Channel, sortPriority=50}.R) {name=LensesMask, nodeColor=<rgba 0 1 0.7422 1>}
  n82 = LocalPosition() {localOrigin=InstancePreSkinning}
  n83 = Subtract(
   A=ShaderStageSwitch(PixelShader=VertexInterpolator(VS=@n82.XYZ).PS, VertexShader=@n82.XYZ)
   B=VectorParameter() {group=PuffinSmile_DotRimLight, parameterName=SphereCenterLocation, defaultValue=<rgba 0 0 0 1>}.RGB
  )
  NamedRerouteDeclaration(None=Normalize(VectorInput=Transform(None=@n83) {transformSourceType=TRANSFORMSOURCE_Local})) {name=SphericalNormal, nodeColor=<rgba 0.2891 1 0 1>}
  NamedRerouteDeclaration(None=ComponentMask(None=@n59) {g=true}) {name=GMap, nodeColor=<rgba 0 1 0.1797 1>}
  n84 = ScalarParameter() {defaultValue=0.025, group=PuffinSmile, parameterName=Lenses_HighlightEdgeHardness, sortPriority=50}
  n85 = VectorParameter() {group=PuffinSmile, parameterName=Lenses_RightEyeOffset, sortPriority=50, defaultValue=<rgba 0.096 0 0 1>}
  n86 = VertexInterpolator(VS=LocalPosition() {localOrigin=InstancePreSkinning}.XYZ)
  n87 = MaterialFunctionCall(Float3=Divide(A=NamedRerouteDeclaration(None=@n86.PS) {name=PS-LocalPos, nodeColor=<rgba 0.3828 0 1 1>}, B=Constant() {r=6}) {constB=30}) {materialFunction=&BreakOutFloat3Components}
  n88 = MaterialFunctionCall(X=@n87.R, Y=@n87.R) {materialFunction=&MakeFloat2}
  n89 = VectorParameter() {group=PuffinSmile, parameterName=Lenses_RightEyeCenter, sortPriority=50, defaultValue=<rgba -0.656 25.8 0 1>}
  n90 = NamedRerouteUsage()
  n91 = MaterialFunctionCall(UVs=@n88.Result, Rotation Center=ComponentMask(None=@n89.RGB) {g=true, r=true}, "Rotation Angle (0-1)"=@n90) {materialFunction=&CustomRotator}
  n92 = Transform(None=Constant3Vector() {constant=<rgba 0 0 1 0>}) {transformSourceType=TRANSFORMSOURCE_View, transformType=TRANSFORM_Local}
  n93 = ComponentMask(None=@n92) {g=true, r=true}
  n94 = Multiply(A=MaterialFunctionCall(Float2=MaterialFunctionCall(Vector or UVs=@n93) {materialFunction=&VectorToRadialValue}.Radial Coordinates) {materialFunction=&BreakOutFloat2Components}.R) {constB=4}
  n95 = ComponentMask(None=@n92) {b=true}
  n96 = Multiply(
   A=ScalarParameter() {defaultValue=0.4, group=02_Anime Parameters, parameterName=Lenses_ViewDirInfluence, sortPriority=50}
   B=AppendVector(A=Add(A=@n94) {constB=-1}, B=Multiply(A=@n95) {constB=-1})
  )
  n97 = VectorParameter() {group=PuffinSmile, parameterName=Lenses_LeftEyeCenter, sortPriority=50, defaultValue=<rgba 0.656 25.8 0 1>}
  n98 = MaterialFunctionCall(UVs=@n88.Result, Rotation Center=ComponentMask(None=@n97.RGB) {g=true, r=true}, "Rotation Angle (0-1)"=@n90) {materialFunction=&CustomRotator}
  n99 = VectorParameter() {group=PuffinSmile, parameterName=Lenses_LeftEyeOffset, sortPriority=50, defaultValue=<rgba -0.232 0 0 1>}
  n100 = NamedRerouteUsage()
  n101 = LinearInterpolate(
   A=Add(A=ComponentMask(None=@n85.RGB) {g=true, r=true}, B=Add(A=@n91.Rotated Values, B=@n96))
   B=Add(A=Add(A=@n98.Rotated Values, B=@n96), B=ComponentMask(None=@n99.RGB) {g=true, r=true})
   Alpha=Step(X=ComponentMask(None=@n100) {r=true}) {constX=0}
  )
  n102 = Subtract(
   A=Sine(None=ComponentMask(None=@n101) {r=true})
   B=OneMinus(None=ScalarParameter() {defaultValue=0.36, group=PuffinSmile, parameterName=Lenses_HighlightThickness, sliderMax=1, sortPriority=50})
  ) {constB=0.6}
  NamedRerouteDeclaration(None=Multiply(A=NamedRerouteUsage(), B=SmoothStep(Max=@n84, Value=@n102) {constMax=0.1})) {name=LensesSpecular, nodeColor=<rgba 0.2188 1 0 1>}
  PixelNormalWS()
 ]}
