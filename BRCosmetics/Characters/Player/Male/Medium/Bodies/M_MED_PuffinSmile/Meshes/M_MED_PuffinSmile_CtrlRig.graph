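# Control Rig for the PuffinSmile male medium cosmetic body, adding procedural body deformation and coat collar, lapel and belt correctives.
# A sequence runs shared DeformRig, then pushes collar targets off head collision spheres, aiming collar bones and blending rotation into neighbours.
# A pose reader and head-driven distance remap adjust back collar bones.
# Upper-arm pose readers drive lapels and front collar, then a weighted parent constraint places the back belt; the neck-rotation loop is unconnected.
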ControlRigBlueprint /BRCosmetics/Characters/Player/Male/Medium/Bodies/M_MED_PuffinSmile/Meshes/M_MED_PuffinSmile_CtrlRig
graphs [
 RigVMFunctionLibrary {links [], nodes []}
 RigVMFunction_Sequence_1_1_ContainedGraph {nodes [
   n1 = RigVMFunctionEntryNode()
   n2 = &RigVMFunction_Sequence(ExecuteContext=@n1.ExecuteContext)
   n3 = &RigVMFunction_Sequence(ExecuteContext=@n2.B)
   n4 = &RigVMFunction_Sequence(ExecuteContext=@n3.B)
   n5 = &RigVMFunction_Sequence(ExecuteContext=@n4.B)
   RigVMFunctionReturnNode(A=@n2.A, B=@n3.A, C=@n4.A, D=@n5.A, E=@n5.B)
  ]}
 RigVMFunction_Sequence_1_ContainedGraph {nodes [
   n1 = RigVMFunctionEntryNode()
   RigVMFunction_Sequence = &RigVMFunction_Sequence(ExecuteContext=@n1.ExecuteContext)
   n2 = &RigVMFunction_Sequence(ExecuteContext=@RigVMFunction_Sequence.B)
   n3 = &RigVMFunction_Sequence(ExecuteContext=@n2.B)
   RigVMFunctionReturnNode(A=@RigVMFunction_Sequence.A, B=@n2.A, C=@n3.A, D=@n3.B)
  ]}
 RigVMFunction_Sequence_2_1_ContainedGraph {nodes [
   n1 = RigVMFunctionEntryNode()
   n2 = &RigVMFunction_Sequence(ExecuteContext=@n1.ExecuteContext)
   n3 = &RigVMFunction_Sequence(ExecuteContext=@n2.B)
   n4 = &RigVMFunction_Sequence(ExecuteContext=@n3.B)
   RigVMFunctionReturnNode(A=@n2.A, B=@n3.A, C=@n4.A, D=@n4.B)
  ]}
 RigVMModel {nodes [
   DISPATCH_RigVMDispatch_ArrayIterator = RigVMDispatchNode(
    ExecuteContext=()
    Array=&RigUnit_ItemArray(Items=((Type=Bone,Name="neck_01"),(Type=Bone,Name="neck_02"),(Type=Bone,Name="head"))).Items
   ) {template="DISPATCH_RigVMDispatch_ArrayIterator(in Array,out Element,out Index,out Count,out Ratio)"}
   n1 = &RigUnit_GetTransform(Space=LocalSpace, bInitial=false, Item=@DISPATCH_RigVMDispatch_ArrayIterator.Element)
   n2 = &RigVMFunction_MathQuaternionToEuler(RotationOrder=ZXY, Value=@n1.Transform.Rotation)
   > &RigUnit_SetTransform(
    Space=LocalSpace
    Value=(Rotation=(X=-0.0029,Y=0.7071,Z=0.0029,W=-0.7071),Translation=(X=0,Y=-1.6873,Z=152.0342),Scale3D=(X=1,Y=1,Z=1))
    Weight=1
    bInitial=False
    Item=@DISPATCH_RigVMDispatch_ArrayIterator.Element
    Value.Rotation=&RigVMFunction_MathQuaternionFromEuler(
     Euler=(X=0,Y=0,Z=0)
     RotationOrder=ZXY
     Euler.X=&RigVMFunction_MathDoubleMul(B=1, A=@n2.Result.X).Result
     Euler.Y=&RigVMFunction_MathDoubleMul(B=1, A=@n2.Result.Y).Result
     Euler.Z=&RigVMFunction_MathDoubleMul(B=0.6, A=@n2.Result.Z).Result
    ).Result
    Value.Translation=@n1.Transform.Translation
    Value.Scale3D=@n1.Transform.Scale3D
   )
   RigUnit_BeginExecution = &RigUnit_BeginExecution()
   RigVMFunction_Sequence = RigVMAggregateNode(ExecuteContext=@RigUnit_BeginExecution.ExecutePin)
   FN_DeformRig_FNC = RigVMFunctionReferenceNode(
    DeformRig_ Curve="(deformRig_l_upperarm_up_alpha_72_#1=False,deformRig_r_upperarm_up_alpha_73_#2=False,deformRig_l_upperarm_fr_alpha_74_#3=False,deformRig_r_upperarm_fr_alpha_75_CB51B8 ...(truncated)"
    DeformRig_Bool="(upperarm_1_#4=true,elbow_3_#5=true,wrist_7_#6=true,groin_8_#7=true,glute_17_#8=true,knee_9_40C408324CA261 ...(truncated)"
    DeformRig_Debug=(upperarm_debug_18_#4=False,wrist_debug_20_#6=False,lat_debug_26_#9=False,groinGlute_debug_27_#7=False)
    DeformRig_Value="(upperarm_up_18_#10=(Rotation=(X=0,Y=0,Z=0,W=1),Translation=(X=0,Y=0,Z=8),Scale3D=(X=0.7,Y=1,Z=1)),upperarm_fwd_fr_19_#11=(Rotation=(X=0,Y=0,Z=0,W=1),Translation=(X=0,Y=10,Z=0),Scale3D=(X=0.5,Y=1,Z= ...(truncated)"
    ExecuteContext=@RigVMFunction_Sequence.A
   )
   RigVMFunction_Sequence_1 = RigVMAggregateNode(ExecuteContext=@RigVMFunction_Sequence.B)
   Spherical Collision = RigVMFunctionReferenceNode(
    Debug=false
    Max Radius=30
    Min Radius=30
    Source=(Type=Bone,Name="dyn_collar_bk_target")
    Source Offset=(X=0,Y=0,Z=0)
    Target=(Type=Bone,Name="head")
    Target Offset=(X=12,Y=19.6,Z=0)
    Weight=1
    ExecuteContext=@RigVMFunction_Sequence_1.A
   )
   > &RigUnit_AimConstraintLocalSpaceOffset(
    AdvancedSettings=(DebugSettings=(bEnabled=false,Scale=10,WorldOffset=(Rotation=(X=0,Y=0,Z=0,W=1),Translation=(X=0,Y=0,Z=0),Scale3D=(X=1,Y=1,Z=1))),RotationOrderForFilter=XZY)
    Child=(Type=Bone,Name="dyn_collar_bk_01")
    Filter=(bX=True,bY=True,bZ=True)
    Parents=((Item=(Type=Bone,Name="dyn_collar_bk_target"),Weight=1))
    Weight=1
    WorldUp=(Target=(X=0,Y=0,Z=20),Kind=Location,Space=(Type=Bone,Name="head"))
   )
   > n3 = &RigUnit_SphericalPoseReader(
    ActiveRegionScaleFactors=(PositiveWidth=1,NegativeWidth=1,PositiveHeight=1,NegativeHeight=1)
    ActiveRegionSize=0.05
    Debug=(bDrawDebug=false,bDraw2D=False,bDrawLocalAxes=False,DebugScale=25,DebugSegments=20,DebugThickness=0.25)
    DriverAxis=(X=1,Y=0,Z=0)
    DriverItem=(Type=Bone,Name="dyn_collar_bk_01")
    FalloffRegionScaleFactors=(PositiveWidth=1,NegativeWidth=1,PositiveHeight=1,NegativeHeight=1)
    FalloffSize=0.3
    FlipHeightScaling=False
    FlipWidthScaling=False
    OptionalParentItem=(Type=Bone,Name="None")
    RotationOffset=(X=0,Y=-145,Z=0)
   )
   > &RigUnit_ModifyTransforms(
    ItemToModify="((Item=(Type=Bone,Name=\"dyn_collar_bk_02\"),Transform=(Rotation=(X=0,Y=0,Z=0,W=1),Translation=(X=-2.9659,Y=0,Z=0),Scale3D=(X=0.9,Y=1,Z=1))),(Item=(Type=Bone,Name=\"dyn_collar_bk_l_02\"),Transform=(Rotation=(X=0,Y=0,Z=0,W=1),Translation=(X=-1.7009,Y=0,Z=0),Sca ...(truncated)"
    WeightMinimum=0
    Weight=@n3.OutputParam
   )
   RigVMFunction_MathQuaternionToEuler = &RigVMFunction_MathQuaternionToEuler(
    RotationOrder=ZYX
    Value=&RigUnit_GetTransform(Item=(Type=Bone,Name="head"), Space=LocalSpace, bInitial=False).Transform.Rotation
   )
   RigVMFunction_MathQuaternionFromEuler = &RigVMFunction_MathQuaternionFromEuler(
    Euler=(X=0,Y=0,Z=0)
    Euler.Z=&RigVMFunction_MathDoubleMul(B=1, A=@RigVMFunction_MathQuaternionToEuler.Result.Z).Result
    Euler.Y=&RigVMFunction_MathDoubleMul(B=1, A=@RigVMFunction_MathQuaternionToEuler.Result.Y).Result
    Euler.X=&RigVMFunction_MathDoubleMul(B=0, A=@RigVMFunction_MathQuaternionToEuler.Result.X).Result
   )
   > RigUnit_SetRotation = &RigUnit_SetRotation(Item=(Type=Bone,Name="dyn_collar_scale_tgt_MSTR"), Weight=1, bInitial=False, Value=@RigVMFunction_MathQuaternionFromEuler.Result)
   > RigUnit_SetTranslation = /Script/ControlRig.RigUnit_SetTranslation(
    Item=(Type=Bone,Name="dyn_collar_scale_tgt_MSTR")
    Space=GlobalSpace
    Weight=1
    bInitial=False
    bPropagateToChildren=True
    Value=&RigUnit_GetTransform(Item=(Type=Bone,Name="head"), bInitial=False).Transform.Translation
   ) {is=struct RigVMUnitNode, template="Set Transform::Execute(in Item,in Space,in bInitial,in Value,in Weight,in bPropagateToChildren)"}
   RigUnit_GetTransform = &RigUnit_GetTransform(Item=(Type=Bone,Name="dyn_collar_scale_tgt"), bInitial=False)
   RigVMFunction_MathVectorDistance = &RigVMFunction_MathVectorDistance(
    A=@RigUnit_GetTransform.Transform.Translation
    B=&RigUnit_GetTransform(Item=(Type=Bone,Name="dyn_collar_bk_01"), bInitial=False).Transform.Translation
   )
   Remap = &RigVMFunction_MathDoubleRemap(SourceMaximum=3.5, SourceMinimum=11.56, TargetMaximum=1, TargetMinimum=0, Value=@RigVMFunction_MathVectorDistance.Result)
   Multiply = &RigVMFunction_MathDoubleMul(A=1, B=@Remap.Result)
   > &RigUnit_ModifyTransforms(
    ItemToModify=((Item=(Type=Bone,Name="dyn_collar_bk_01"),Transform=(Rotation=(X=0,Y=0,Z=0,W=1),Translation=(X=-0.8,Y=0,Z=0),Scale3D=(X=1,Y=1,Z=1))))
    WeightMinimum=0
    Weight=@Multiply.Result
   )
   > &RigUnit_ModifyTransforms(
    ItemToModify=((Item=(Type=Bone,Name="dyn_collar_bk_02"),Transform=(Rotation=(X=0,Y=0,Z=0,W=1),Translation=(X=-6,Y=0,Z=0),Scale3D=(X=1,Y=1,Z=1))))
    WeightMinimum=0
    Weight=@Multiply.Result
   )
   > &RigUnit_ModifyTransforms(
    ItemToModify=((Item=(Type=Bone,Name="dyn_collar_bk_l_01"),Transform=(Rotation=(X=0,Y=0,Z=0,W=1),Translation=(X=-0.8,Y=0,Z=0),Scale3D=(X=1,Y=1,Z=1))))
    WeightMinimum=0
    Weight=@Multiply.Result
   )
   > &RigUnit_ModifyTransforms(
    ItemToModify=((Item=(Type=Bone,Name="dyn_collar_bk_l_02"),Transform=(Rotation=(X=0,Y=0,Z=0,W=1),Translation=(X=-4,Y=0,Z=0),Scale3D=(X=1,Y=1,Z=1))))
    WeightMinimum=0
    Weight=@Multiply.Result
   )
   > &RigUnit_ModifyTransforms(
    ItemToModify=((Item=(Type=Bone,Name="dyn_collar_bk_r_01"),Transform=(Rotation=(X=0,Y=0,Z=0,W=1),Translation=(X=-0.8,Y=0,Z=0),Scale3D=(X=1,Y=1,Z=1))))
    WeightMinimum=0
    Weight=@Multiply.Result
   )
   > &RigUnit_ModifyTransforms(
    ItemToModify=((Item=(Type=Bone,Name="dyn_collar_bk_r_02"),Transform=(Rotation=(X=0,Y=0,Z=0,W=1),Translation=(X=-4,Y=0,Z=0),Scale3D=(X=1,Y=1,Z=1))))
    WeightMinimum=0
    Weight=@Multiply.Result
   )
   Spherical Collision_1 = RigVMFunctionReferenceNode(
    Debug=false
    Max Radius=30
    Min Radius=30
    Source=(Type=Bone,Name="dyn_collar_bk_l_target")
    Source Offset=(X=0,Y=0,Z=0)
    Target=(Type=Bone,Name="head")
    Target Offset=(X=7,Y=15.3,Z=15)
    Weight=1
    ExecuteContext=@RigVMFunction_Sequence_1.B
   )
   > &RigUnit_AimConstraintLocalSpaceOffset(
    AdvancedSettings=(DebugSettings=(bEnabled=false,Scale=10,WorldOffset=(Rotation=(X=0,Y=0,Z=0,W=1),Translation=(X=0,Y=0,Z=0),Scale3D=(X=1,Y=1,Z=1))),RotationOrderForFilter=XZY)
    Child=(Type=Bone,Name="dyn_collar_bk_l_01")
    Filter=(bX=True,bY=True,bZ=True)
    Parents=((Item=(Type=Bone,Name="dyn_collar_bk_l_target"),Weight=1))
    Weight=1
    WorldUp=(Target=(X=0,Y=0,Z=20),Kind=Location,Space=(Type=Bone,Name="head"))
   )
   > &RigUnit_RotationConstraintLocalSpaceOffset(
    AdvancedSettings=(InterpolationType=Average,RotationOrderForFilter=XZY)
    Child=(Type=Bone,Name="dyn_collar_mid_l_MSTR")
    Filter=(bX=True,bY=True,bZ=True)
    Parents=((Item=(Type=Bone,Name="dyn_collar_mid_l_MSTR"),Weight=1),(Item=(Type=Bone,Name="dyn_collar_bk_l_01"),Weight=0.5))
    Weight=1
   )
   Spherical Collision_1_2 = RigVMFunctionReferenceNode(
    Debug=false
    Max Radius=30
    Min Radius=30
    Source=(Type=Bone,Name="dyn_collar_mid_l_target")
    Source Offset=(X=0,Y=0,Z=0)
    Target=(Type=Bone,Name="head")
    Target Offset=(X=7,Y=0,Z=23)
    Weight=1
    ExecuteContext=@RigVMFunction_Sequence_1.C
   )
   > &RigUnit_AimConstraintLocalSpaceOffset(
    Child=(Type=Bone,Name="dyn_collar_mid_l_01")
    Filter=(bX=True,bY=True,bZ=True)
    Parents=((Item=(Type=Bone,Name="dyn_collar_mid_l_target"),Weight=1))
    Weight=1
    WorldUp=(Target=(X=0,Y=0,Z=20),Kind=Location,Space=(Type=Bone,Name="head"))
   )
   > &RigUnit_RotationConstraintLocalSpaceOffset(
    AdvancedSettings=(InterpolationType=Average,RotationOrderForFilter=XZY)
    Child=(Type=Bone,Name="dyn_collar_fr_l_MSTR")
    Filter=(bX=True,bY=True,bZ=True)
    Parents=((Item=(Type=Bone,Name="dyn_collar_fr_l_MSTR"),Weight=1),(Item=(Type=Bone,Name="dyn_collar_mid_l_01"),Weight=0.5))
    Weight=1
   )
   Spherical Collision_1_1 = RigVMFunctionReferenceNode(
    Debug=false
    Max Radius=30
    Min Radius=30
    Source=(Type=Bone,Name="dyn_collar_bk_r_target")
    Source Offset=(X=0,Y=0,Z=0)
    Target=(Type=Bone,Name="head")
    Target Offset=(X=7,Y=15.3,Z=-15)
    Weight=1
    ExecuteContext=@RigVMFunction_Sequence_1.D
   )
   > &RigUnit_AimConstraintLocalSpaceOffset(
    Child=(Type=Bone,Name="dyn_collar_bk_r_01")
    Filter=(bX=True,bY=True,bZ=True)
    Parents=((Item=(Type=Bone,Name="dyn_collar_bk_r_target"),Weight=1))
    Weight=1
    WorldUp=(Target=(X=0,Y=0,Z=20),Kind=Location,Space=(Type=Bone,Name="head"))
   )
   > &RigUnit_RotationConstraintLocalSpaceOffset(
    AdvancedSettings=(InterpolationType=Average,RotationOrderForFilter=XZY)
    Child=(Type=Bone,Name="dyn_collar_mid_r_MSTR")
    Filter=(bX=true,bY=True,bZ=True)
    Parents=((Item=(Type=Bone,Name="dyn_collar_mid_r_MSTR"),Weight=1),(Item=(Type=Bone,Name="dyn_collar_bk_r_01"),Weight=0.5))
    Weight=1
    bMaintainOffset=true
   )
   Spherical Collision_1_2_1 = RigVMFunctionReferenceNode(
    Debug=false
    Max Radius=30
    Min Radius=30
    Source=(Type=Bone,Name="dyn_collar_mid_r_target")
    Source Offset=(X=0,Y=0,Z=0)
    Target=(Type=Bone,Name="head")
    Target Offset=(X=7,Y=0,Z=-23)
    Weight=1
    ExecuteContext=@RigVMFunction_Sequence_1.E
   )
   > &RigUnit_AimConstraintLocalSpaceOffset(
    Child=(Type=Bone,Name="dyn_collar_mid_r_01")
    Filter=(bX=True,bY=True,bZ=True)
    Parents=((Item=(Type=Bone,Name="dyn_collar_mid_r_target"),Weight=1))
    Weight=1
    WorldUp=(Target=(X=0,Y=0,Z=20),Kind=Location,Space=(Type=Bone,Name="head"))
   )
   > &RigUnit_RotationConstraintLocalSpaceOffset(
    AdvancedSettings=(InterpolationType=Average,RotationOrderForFilter=XZY)
    Child=(Type=Bone,Name="dyn_collar_fr_r_MSTR")
    Filter=(bX=True,bY=True,bZ=True)
    Parents=((Item=(Type=Bone,Name="dyn_collar_fr_r_MSTR"),Weight=1),(Item=(Type=Bone,Name="dyn_collar_mid_r_01"),Weight=0.5))
    Weight=1
   )
   RigVMFunction_Sequence_2 = RigVMAggregateNode(ExecuteContext=@RigVMFunction_Sequence.C)
   n4 = &RigUnit_SphericalPoseReader(
    ActiveRegionScaleFactors=(PositiveWidth=0.05,NegativeWidth=0.05,PositiveHeight=0.5,NegativeHeight=0.5)
    ActiveRegionSize=0.4
    Debug=(bDrawDebug=false,bDraw2D=False,bDrawLocalAxes=False,DebugScale=25,DebugSegments=20,DebugThickness=0.25)
    DriverAxis=(X=-1,Y=0,Z=0)
    DriverItem=(Type=Bone,Name="upperarm_r")
    FalloffRegionScaleFactors=(PositiveWidth=0.8,NegativeWidth=0.8,PositiveHeight=0.8,NegativeHeight=0.8)
    FalloffSize=0.5
    FlipHeightScaling=false
    FlipWidthScaling=false
    OptionalParentItem=(Type=Bone,Name="None")
    RotationOffset=(X=-82.298,Y=21.7645,Z=14.6491)
    ExecutePin=@RigVMFunction_Sequence_2.A
   )
   > &RigUnit_ModifyTransforms(
    ItemToModify=((Item=(Type=Bone,Name="dyn_lapel_r_MSTR"),Transform=(Rotation=(X=-0.1254,Y=-0.4369,Z=0.0381,W=0.8899),Translation=(X=-0.6173,Y=0.204,Z=0.6674),Scale3D=(X=1,Y=1,Z=1))))
    WeightMinimum=0
    Weight=@n4.OutputParam
   )
   n5 = &RigUnit_SphericalPoseReader(
    ActiveRegionScaleFactors=(PositiveWidth=0.02,NegativeWidth=0.02,PositiveHeight=0.02,NegativeHeight=0.02)
    ActiveRegionSize=0.4
    Debug=(bDrawDebug=false,bDraw2D=False,bDrawLocalAxes=False,DebugScale=25,DebugSegments=20,DebugThickness=0.25)
    DriverAxis=(X=-1,Y=0,Z=0)
    DriverItem=(Type=Bone,Name="upperarm_r")
    FalloffRegionScaleFactors=(PositiveWidth=0.8,NegativeWidth=0.8,PositiveHeight=0.8,NegativeHeight=0.8)
    FalloffSize=0.5
    FlipHeightScaling=false
    FlipWidthScaling=false
    OptionalParentItem=(Type=Bone,Name="None")
    RotationOffset=(X=-117.5991,Y=21.7645,Z=16.1917)
    ExecutePin=@RigVMFunction_Sequence_2.B
   )
   > &RigUnit_ModifyTransforms(
    ItemToModify=((Item=(Type=Bone,Name="dyn_collar_fr_r_MSTR"),Transform=(Rotation=(X=-0.3307,Y=-0.3436,Z=0.0646,W=0.8766),Translation=(X=0,Y=0,Z=0),Scale3D=(X=1,Y=1,Z=1))))
    WeightMinimum=0
    Weight=@n5.OutputParam
   )
   n6 = &RigUnit_SphericalPoseReader(
    ActiveRegionScaleFactors=(PositiveWidth=0.05,NegativeWidth=0.05,PositiveHeight=0.1,NegativeHeight=0.1)
    ActiveRegionSize=0.4
    Debug=(bDrawDebug=false,bDraw2D=False,bDrawLocalAxes=False,DebugScale=25,DebugSegments=20,DebugThickness=0.25)
    DriverAxis=(X=1,Y=0,Z=0)
    DriverItem=(Type=Bone,Name="upperarm_l")
    FalloffRegionScaleFactors=(PositiveWidth=0.8,NegativeWidth=0.8,PositiveHeight=0.8,NegativeHeight=0.8)
    FalloffSize=0.5
    FlipHeightScaling=false
    FlipWidthScaling=false
    OptionalParentItem=(Type=Bone,Name="None")
    RotationOffset=(X=78.8661,Y=16,Z=2.7685)
    ExecutePin=@RigVMFunction_Sequence_2.C
   )
   > &RigUnit_ModifyTransforms(
    ItemToModify=((Item=(Type=Bone,Name="dyn_lapel_l_MSTR"),Transform=(Rotation=(X=0.126,Y=-0.2515,Z=0.0658,W=0.9574),Translation=(X=0,Y=0,Z=0),Scale3D=(X=1,Y=1,Z=1))))
    WeightMinimum=0
    Weight=@n6.OutputParam
   )
   n7 = &RigUnit_SphericalPoseReader(
    ActiveRegionScaleFactors=(PositiveWidth=0.02,NegativeWidth=0.02,PositiveHeight=0.02,NegativeHeight=0.02)
    ActiveRegionSize=0.4
    Debug=(bDrawDebug=false,bDraw2D=False,bDrawLocalAxes=False,DebugScale=25,DebugSegments=20,DebugThickness=0.25)
    DriverAxis=(X=1,Y=0,Z=0)
    DriverItem=(Type=Bone,Name="upperarm_l")
    FalloffRegionScaleFactors=(PositiveWidth=0.8,NegativeWidth=0.8,PositiveHeight=0.8,NegativeHeight=0.8)
    FalloffSize=0.5
    FlipHeightScaling=false
    FlipWidthScaling=false
    OptionalParentItem=(Type=Bone,Name="None")
    RotationOffset=(X=71.2715,Y=21.7645,Z=-10.2732)
    ExecutePin=@RigVMFunction_Sequence_2.D
   )
   > &RigUnit_ModifyTransforms(
    ItemToModify=((Item=(Type=Bone,Name="dyn_collar_fr_l_MSTR"),Transform=(Rotation=(X=0.2376,Y=-0.2781,Z=-0.0291,W=0.9302),Translation=(X=0,Y=0,Z=0),Scale3D=(X=1,Y=1,Z=1))))
    WeightMinimum=0
    Weight=@n7.OutputParam
   )
   &RigUnit_ParentConstraint(
    AdvancedSettings=(RotationOrderForFilter=XZY)
    Child=(Type=Bone,Name="dyn_belt_bk_MSTR")
    Parents="((Item=(Type=Bone,Name=\"spine_04\"),Weight=0.0203),(Item=(Type=Bone,Name=\"dyn_coat_bk_l_01\"),Weight=0.0217),(Item=(Type=Bone,Name=\"spine_03\"),Weight=0.1249),(Item=(Type=Bone,Name=\"dyn_coat_bk_01\"),Weight=0.1681),(Item=(Type=Bone,Name=\"spine_02\"),Weight=0.17 ...(truncated)"
    Weight=1
    ExecutePin=@RigVMFunction_Sequence.D
   )
   &RigVMFunction_MathDoubleMul(
    A=1
    B=&RigVMFunction_MathDoubleRemap(SourceMaximum=13.2, SourceMinimum=5.8, TargetMaximum=1, TargetMinimum=0, Value=@RigVMFunction_MathVectorDistance.Result).Result
   )
  ]}
]
hierarchy [
 root {loc=<0 0 0>, rot=<0 0 0 1>}
 pelvis {parent=root, loc=<0 -0.7071 93.1191>, rot=<0.0127 0.707 -0.0127 -0.707>}
 spine_01 {parent=pelvis, loc=<6.7494 -0.6388 0>, rot=<0 0 0.034 -0.9994>}
 spine_02 {parent=spine_01, loc=<7.252 -0.035 0>, rot=<0 0 0.0238 -0.9997>}
 spine_03 {parent=spine_02, loc=<6.0106 0 0>, rot=<0 0 0.0149 -0.9999>}
 spine_04 {parent=spine_03, loc=<8.1023 0 0>, rot=<0 0 -0.0132 -0.9999>}
 spine_05 {parent=spine_04, loc=<10.0142 0 0>, rot=<0 0 -0.0015 -1>}
 attach_backpack {parent=spine_05, loc=<-0.123 -13.473 0>, rot=<-0.0093 -0.707 0.0093 -0.707>}
 attach_cape {parent=spine_05, loc=<9.017 -11.714 0>, rot=<0.0616 -0.7044 -0.0616 -0.7044>}
 neck_01 {parent=spine_05, loc=<11.3138 0.1529 0>, rot=<0 0 -0.2003 -0.9797>}
 neck_02 {parent=neck_01, loc=<4.5627 0 0>, rot=<0 0 0.0504 -0.9987>}
 head {parent=neck_02, loc=<5.3914 0 0>, rot=<0 0 0.115 -0.9934>}
 faceAttach {parent=head, loc=<0.9213 0.0076 0>, rot=<0.0044 -0.3827 -0.0044 -0.9239>, scale=<0.89 0.89 0.89>}
 C_jaw {parent=faceAttach, loc=<0.4365 0.6972 0.4651>, rot=<0.0044 -0.3827 -0.0044 -0.9239>}
 clavicle_l {parent=spine_05, loc=<4.3575 -0.1466 -2.3223>, rot=<0.0465 -0.7289 -0.009 -0.6829>}
 upperarm_l {parent=clavicle_l, loc=<15.5111 0 0>, rot=<0.0226 -0.3234 -0.0088 -0.9459>}
 lowerarm_l {parent=upperarm_l, loc=<26.6514 0 0>, rot=<0.0391 -0.0099 -0.2453 -0.9686>}
 hand_l {parent=lowerarm_l, loc=<22.3491 0 0>, rot=<-0.7081 -0.0507 0.0975 -0.6975>}
 thumb_01_l {parent=hand_l, loc=<1.3118 -0.3191 -2.459>, rot=<0.6546 -0.4042 0.0664 -0.6354>}
 thumb_02_l {parent=thumb_01_l, loc=<5.9336 -0.0448 0.0401>, rot=<-0.0037 0.0213 0.298 -0.9543>}
 thumb_03_l {parent=thumb_02_l, loc=<2.2223 0.3682 -0.0907>, rot=<0.0118 -0.013 0.0066 -0.9998>}
 index_metacarpal_l {parent=hand_l, loc=<1.6535 0.0183 -2.96>, rot=<0.0158 -0.0124 -0.0173 -0.9996>}
 index_01_l {parent=index_metacarpal_l, loc=<8.3008 0.0357 -0.0524>, rot=<-0.0044 0.0485 0.0924 -0.9945>}
 index_02_l {parent=index_01_l, loc=<3.3587 -0.3004 0.0452>, rot=<-0.002 0.0027 0.3225 -0.9466>}
 index_03_l {parent=index_02_l, loc=<2.5578 -0.123 -0.0252>, rot=<-0.002 0.0067 0.2558 -0.9667>}
 middle_metacarpal_l {parent=hand_l, loc=<2.1856 -0.1725 -0.5984>, rot=<0.006 0.0168 -0.0329 -0.9993>}
 middle_01_l {parent=middle_metacarpal_l, loc=<7.3275 -0.2447 0.0129>, rot=<-0.0208 0.0042 0.0858 -0.9961>}
 middle_02_l {parent=middle_01_l, loc=<4.0123 0.1401 0.1175>, rot=<-0.0218 0.0258 0.3512 -0.9357>}
 middle_03_l {parent=middle_02_l, loc=<2.6077 0.0534 0.1373>, rot=<0.0082 0.0188 0.1708 -0.9851>}
 ring_metacarpal_l {parent=hand_l, loc=<2.2074 -0.3625 1.3638>, rot=<-0.0284 0.0722 -0.0458 -0.9959>}
 ring_01_l {parent=ring_metacarpal_l, loc=<7.0295 -0.6197 -0.1446>, rot=<-0.0489 -0.0252 0.1892 -0.9804>}
 ring_02_l {parent=ring_01_l, loc=<3.6174 0.4647 -0.1027>, rot=<0 0 0.2806 -0.9598>}
 ring_03_l {parent=ring_02_l, loc=<2.4152 0.114 0.0199>, rot=<0 0 0.1189 -0.9929>}
 pinky_metacarpal_l {parent=hand_l, loc=<1.7616 -0.9753 3.3814>, rot=<-0.0397 0.0814 -0.069 -0.9935>}
 pinky_01_l {parent=pinky_metacarpal_l, loc=<7.1709 -0.4831 -0.1094>, rot=<-0.0654 -0.0035 0.2242 -0.9723>}
 pinky_02_l {parent=pinky_01_l, loc=<2.8274 0.0893 -0.1043>, rot=<0.0037 0.0011 0.2854 -0.9584>}
 pinky_03_l {parent=pinky_02_l, loc=<2.2046 0.0307 -0.0078>, rot=<-0.0002 0.0098 0.1445 -0.9895>}
 deform_wrist_dn_base_l {parent=hand_l, loc=<0 -2.2342 0>, rot=<0 0 0 -1>}
 deform_wrist_dn_l {parent=deform_wrist_dn_base_l, loc=<0 0 0>, rot=<0 0 0 1>}
 deform_wrist_up_base_l {parent=hand_l, loc=<0 2.2682 0>, rot=<0 0 0 -1>}
 deform_wrist_up_l {parent=deform_wrist_up_base_l, loc=<0 0 0>, rot=<0 0 0 1>}
 lowerarm_twist_02_l {parent=lowerarm_l, loc=<9.6199 0 0>, rot=<0 0 0 -1>}
 lowerarm_twist_01_l {parent=lowerarm_l, loc=<17.8549 0 0>, rot=<0 0 0 -1>}
 deform_elbow_in_base_l {parent=lowerarm_l, loc=<0.2112 5.259 0.8274>, rot=<0 0 0 -1>}
 deform_elbow_in_l {parent=deform_elbow_in_base_l, loc=<0 0 0>, rot=<0 0 0 1>}
 deform_elbow_out_base_l {parent=lowerarm_l, loc=<0 -4 0>, rot=<0 0 0 -1>}
 deform_elbow_out_l {parent=deform_elbow_out_base_l, loc=<0 0 0>, rot=<0 0 0 1>}
 dyn_sleeve_l_01 {parent=lowerarm_l, loc=<7.4711 0 0>, rot=<0 0.046 0 0.9989>}
 dyn_sleeve_l_02 {parent=dyn_sleeve_l_01, loc=<11.396 0 0>, rot=<0 0 0 1>}
 upperarm_twist_01_l {parent=upperarm_l, loc=<0.0041 0 0>, rot=<0 0 0 -1>}
 deform_upperarm_bk_base_l {parent=upperarm_twist_01_l, loc=<1.0813 -7.2193 -1.9453>, rot=<0 0 0 1>}
 deform_upperarm_bk_l {parent=deform_upperarm_bk_base_l, loc=<0 0 0>, rot=<0 0 0 1>}
 deform_upperarm_fr_base_l {parent=upperarm_twist_01_l, loc=<-0.3691 7.0506 -2.765>, rot=<0 0 0 1>}
 deform_upperarm_fr_l {parent=deform_upperarm_fr_base_l, loc=<0 0 0>, rot=<0 0 0 1>}
 deform_upperarm_up_base_l {parent=upperarm_twist_01_l, loc=<-3.2216 -1.1391 3.3311>, rot=<0 0 0 1>}
 deform_upperarm_up_l {parent=deform_upperarm_up_base_l, loc=<0 0 0>, rot=<0 0 0 1>}
 upperarm_twist_02_l {parent=upperarm_l, loc=<13.7941 0 0>, rot=<0 0 0 -1>}
 clavicle_r {parent=spine_05, loc=<4.3575 -0.1466 2.3223>, rot=<0.7289 0.0465 -0.6829 0.009>}
 upperarm_r {parent=clavicle_r, loc=<-15.5111 0 0>, rot=<0.0226 -0.3234 -0.0088 -0.9459>}
 lowerarm_r {parent=upperarm_r, loc=<-26.6514 0 0>, rot=<0.0391 -0.0099 -0.2453 -0.9686>}
 hand_r {parent=lowerarm_r, loc=<-22.3491 0 0>, rot=<-0.7081 -0.0507 0.0975 -0.6975>}
 thumb_01_r {parent=hand_r, loc=<-1.3118 0.3191 2.459>, rot=<0.6546 -0.4042 0.0664 -0.6354>}
 thumb_02_r {parent=thumb_01_r, loc=<-5.9336 0.0448 -0.0401>, rot=<-0.0037 0.0213 0.298 -0.9543>}
 thumb_03_r {parent=thumb_02_r, loc=<-2.2223 -0.3682 0.0907>, rot=<0.0118 -0.013 0.0066 -0.9998>}
 index_metacarpal_r {parent=hand_r, loc=<-1.6535 -0.0183 2.96>, rot=<0.0158 -0.0124 -0.0173 -0.9996>}
 index_01_r {parent=index_metacarpal_r, loc=<-8.3008 -0.0357 0.0524>, rot=<-0.0044 0.0485 0.0924 -0.9945>}
 index_02_r {parent=index_01_r, loc=<-3.3587 0.3004 -0.0452>, rot=<-0.002 0.0027 0.3225 -0.9466>}
 index_03_r {parent=index_02_r, loc=<-2.5578 0.123 0.0252>, rot=<-0.002 0.0067 0.2558 -0.9667>}
 middle_metacarpal_r {parent=hand_r, loc=<-2.1856 0.1725 0.5984>, rot=<0.006 0.0168 -0.0329 -0.9993>}
 middle_01_r {parent=middle_metacarpal_r, loc=<-7.3275 0.2447 -0.0129>, rot=<-0.0208 0.0042 0.0858 -0.9961>}
 middle_02_r {parent=middle_01_r, loc=<-4.0123 -0.1401 -0.1175>, rot=<-0.0218 0.0258 0.3512 -0.9357>}
 middle_03_r {parent=middle_02_r, loc=<-2.6077 -0.0534 -0.1373>, rot=<0.0082 0.0188 0.1708 -0.9851>}
 ring_metacarpal_r {parent=hand_r, loc=<-2.2074 0.3625 -1.3638>, rot=<-0.0284 0.0722 -0.0458 -0.9959>}
 ring_01_r {parent=ring_metacarpal_r, loc=<-7.0295 0.6197 0.1446>, rot=<-0.0489 -0.0252 0.1892 -0.9804>}
 ring_02_r {parent=ring_01_r, loc=<-3.6174 -0.4647 0.1027>, rot=<0 0 0.2806 -0.9598>}
 ring_03_r {parent=ring_02_r, loc=<-2.4152 -0.114 -0.0199>, rot=<0 0 0.1189 -0.9929>}
 pinky_metacarpal_r {parent=hand_r, loc=<-1.7616 0.9753 -3.3814>, rot=<-0.0397 0.0814 -0.069 -0.9935>}
 pinky_01_r {parent=pinky_metacarpal_r, loc=<-7.1709 0.4831 0.1094>, rot=<-0.0654 -0.0035 0.2242 -0.9723>}
 pinky_02_r {parent=pinky_01_r, loc=<-2.8274 -0.0893 0.1043>, rot=<0.0037 0.0011 0.2854 -0.9584>}
 pinky_03_r {parent=pinky_02_r, loc=<-2.2046 -0.0307 0.0078>, rot=<-0.0002 0.0098 0.1445 -0.9895>}
 deform_wrist_dn_base_r {parent=hand_r, loc=<-0.0004 2.2345 -0.0001>, rot=<0 0 0 -1>}
 deform_wrist_dn_r {parent=deform_wrist_dn_base_r, loc=<0 0 0>, rot=<0 0 0 1>}
 deform_wrist_up_base_r {parent=hand_r, loc=<0.0003 -2.2684 0.0001>, rot=<0 0 0 -1>}
 deform_wrist_up_r {parent=deform_wrist_up_base_r, loc=<0 0 0>, rot=<0 0 0 1>}
 lowerarm_twist_02_r {parent=lowerarm_r, loc=<-9.6199 0 0>, rot=<0 0 0 -1>}
 lowerarm_twist_01_r {parent=lowerarm_r, loc=<-17.8549 0 0>, rot=<0 0 0 -1>}
 deform_elbow_in_base_r {parent=lowerarm_r, loc=<-0.2111 -5.2591 -0.8275>, rot=<0 0 0 -1>}
 deform_elbow_in_r {parent=deform_elbow_in_base_r, loc=<-0.0001 0 0.0002>, rot=<0 0 0 1>}
 deform_elbow_out_base_r {parent=lowerarm_r, loc=<0 4 -0.0001>, rot=<0 0 0 -1>}
 deform_elbow_out_r {parent=deform_elbow_out_base_r, loc=<0 0 0>, rot=<0 0 0 1>}
 dyn_sleeve_r_01 {parent=lowerarm_r, loc=<-7.4711 0 0>, rot=<0 0.058 0 0.9983>}
 dyn_sleeve_r_02 {parent=dyn_sleeve_r_01, loc=<-11.396 0 0>, rot=<0 0 0 1>}
 upperarm_twist_01_r {parent=upperarm_r, loc=<-0.0041 0 0>, rot=<0 0 0 -1>}
 deform_upperarm_bk_base_r {parent=upperarm_twist_01_r, loc=<-1.0814 7.2192 1.9455>, rot=<0 0 0 1>}
 deform_upperarm_bk_r {parent=deform_upperarm_bk_base_r, loc=<0 0 0>, rot=<0 0 0 1>}
 deform_upperarm_fr_base_r {parent=upperarm_twist_01_r, loc=<0.3689 -7.0506 2.7652>, rot=<0 0 0 1>}
 deform_upperarm_fr_r {parent=deform_upperarm_fr_base_r, loc=<0 0 0>, rot=<0 0 0 1>}
 deform_upperarm_up_base_r {parent=upperarm_twist_01_r, loc=<3.2214 1.1391 -3.3308>, rot=<0 0 0 1>}
 deform_upperarm_up_r {parent=deform_upperarm_up_base_r, loc=<0 0 0>, rot=<0 0 0 -1>}
 upperarm_twist_02_r {parent=upperarm_r, loc=<-13.7941 0 0>, rot=<0 0 0 -1>}
 dyn_lapel_l_MSTR {parent=spine_05, loc=<2.0138 9.3764 -9.3744>, rot=<0.2472 -0.2697 -0.5417 -0.7568>}
 dyn_lapel_l {parent=dyn_lapel_l_MSTR, loc=<1.2461 0 0.631>, rot=<0 -0.1113 0 -0.9938>}
 dyn_collar_fr_l_MSTR {parent=spine_05, loc=<7.1485 6.4733 -8.5539>, rot=<0.2404 -0.1934 -0.2245 -0.9244>}
 dyn_collar_fr_l_01 {parent=dyn_collar_fr_l_MSTR, loc=<1.7935 0 0.3981>, rot=<0 -0.0851 0 -0.9964>}
 dyn_collar_fr_l_02 {parent=dyn_collar_fr_l_01, loc=<7.1903 0 0>, rot=<0 0 0 1>}
 dyn_collar_bk_MSTR {parent=spine_05, loc=<14.1823 -4.3124 0>, rot=<-0.7071 0 0 0.7071>}
 dyn_collar_bk_01 {parent=dyn_collar_bk_MSTR, loc=<2.0872 0 -0.3685>, rot=<0 0.1142 0 0.9935>}
 dyn_collar_bk_02 {parent=dyn_collar_bk_01, loc=<5.8927 0 0>, rot=<0 0 0 1>}
 dyn_collar_bk_l_MSTR {parent=spine_05, loc=<14.6159 -2.9579 -4.698>, rot=<-0.4688 0.0757 0.0588 0.8781>}
 dyn_collar_bk_l_01 {parent=dyn_collar_bk_l_MSTR, loc=<1.4297 0 0.5506>, rot=<0 0.0355 0 0.9994>}
 dyn_collar_bk_l_02 {parent=dyn_collar_bk_l_01, loc=<5.639 0 0.0168>, rot=<0 0.0331 0 0.9995>}
 dyn_collar_mid_l_MSTR {parent=spine_05, loc=<11.4429 2.2786 -7.5764>, rot=<-0.3003 0.0796 0.1519 0.9383>}
 dyn_collar_mid_l_01 {parent=dyn_collar_mid_l_MSTR, loc=<1.2579 0 0.5801>, rot=<0 0.0982 0 0.9952>}
 dyn_collar_mid_l_02 {parent=dyn_collar_mid_l_01, loc=<6.7907 0 0>, rot=<0 0 0 1>}
 dyn_collar_fr_r_MSTR {parent=spine_05, loc=<7.1485 6.4733 8.5539>, rot=<0.9244 0.2245 0.1934 -0.2404>}
 dyn_collar_fr_r_01 {parent=dyn_collar_fr_r_MSTR, loc=<1.7935 0 0.3981>, rot=<0 0.0851 0 0.9964>}
 dyn_collar_fr_r_02 {parent=dyn_collar_fr_r_01, loc=<7.1903 0 0>, rot=<0 0 0 1>}
 dyn_lapel_r_MSTR {parent=spine_05, loc=<2.0138 9.3764 9.3744>, rot=<0.7568 0.5417 0.2697 -0.2472>}
 dyn_lapel_r {parent=dyn_lapel_r_MSTR, loc=<1.2461 0 0.631>, rot=<0 0.1113 0 0.9938>}
 dyn_collar_bk_r_MSTR {parent=spine_05, loc=<14.6159 -2.9579 4.698>, rot=<0.8781 0.0588 0.0757 -0.4688>}
 dyn_collar_bk_r_01 {parent=dyn_collar_bk_r_MSTR, loc=<1.4297 0 0.5506>, rot=<0 0.0355 0 0.9994>}
 dyn_collar_bk_r_02 {parent=dyn_collar_bk_r_01, loc=<5.639 0 0.0168>, rot=<0 0.0331 0 0.9995>}
 dyn_collar_mid_r_MSTR {parent=spine_05, loc=<11.4429 2.2786 7.5764>, rot=<0.9383 0.1519 0.0796 -0.3003>}
 dyn_collar_mid_r_01 {parent=dyn_collar_mid_r_MSTR, loc=<1.2579 0 0.5801>, rot=<0 0.0982 0 0.9952>}
 dyn_collar_mid_r_02 {parent=dyn_collar_mid_r_01, loc=<6.7907 0 0>, rot=<0 0 0 1>}
 dyn_collar_fr_l_target {parent=spine_05, loc=<14.2641 8.8966 -13.3516>, rot=<-0.2586 0.2713 0.2032 0.9045>}
 dyn_collar_bk_target {parent=spine_05, loc=<22.0086 -6.0179 0>, rot=<-0.7025 0.0807 -0.0807 0.7025>}
 dyn_collar_bk_l_target {parent=spine_05, loc=<21.555 -2.5917 -5.9329>, rot=<-0.4718 0.1357 0.0265 0.8708>}
 dyn_collar_mid_l_target {parent=spine_05, loc=<18.8513 3.7174 -10.0849>, rot=<-0.3138 0.1714 0.1217 0.9259>}
 dyn_collar_fr_r_target {parent=spine_05, loc=<14.2641 8.8966 13.3516>, rot=<0.9045 0.2032 0.2713 -0.2586>}
 dyn_collar_bk_r_target {parent=spine_05, loc=<21.555 -2.5917 5.9329>, rot=<0.8708 0.0265 0.1357 -0.4718>}
 dyn_collar_mid_r_target {parent=spine_05, loc=<18.8513 3.7174 10.0849>, rot=<0.9259 0.1217 0.1714 -0.3138>}
 dyn_collar_scale_tgt_MSTR {parent=spine_05, loc=<20.6573 3.5489 0>, rot=<0 0 -0.036 -0.9994>}
 dyn_collar_scale_tgt {parent=dyn_collar_scale_tgt_MSTR, loc=<5.9519 -11.707 0>, rot=<0.6991 -0.106 0.106 -0.6991>}
 deform_pec_aim_base_l {parent=spine_04, loc=<7.8728 12.6388 -2.4123>, rot=<0.0102 -0.4602 0.3979 -0.7936>}
 deform_pec_aim_l {parent=deform_pec_aim_base_l, loc=<0 0 0>, rot=<0 0 0 1>}
 deform_pec_base_l {parent=deform_pec_aim_l, loc=<10.766 0 0>, rot=<0 0 0 1>}
 deform_pec_l {parent=deform_pec_base_l, loc=<0.055 0 0>, rot=<0 0 0 1>}
 deform_pec_aim_base_r {parent=spine_04, loc=<7.8723 12.6388 2.4123>, rot=<-0.4602 -0.0102 0.7936 0.3979>}
 deform_pec_aim_r {parent=deform_pec_aim_base_r, loc=<0 0 0>, rot=<0 0 0 1>}
 deform_pec_base_r {parent=deform_pec_aim_r, loc=<-10.7662 -0.0006 -0.0007>, rot=<0 0 0 1>}
 deform_pec_r {parent=deform_pec_base_r, loc=<0 0 0>, rot=<0 0 0 1>}
 dyn_coat_l_MSTR {parent=spine_04, loc=<2.951 10.5522 -12.2799>, rot=<0.07 -0.9056 -0.418 -0.0183>}
 dyn_coat_l_01 {parent=dyn_coat_l_MSTR, loc=<0.3 0 0>, rot=<0 0 0 1>}
 dyn_coat_l_02 {parent=dyn_coat_l_01, loc=<11.567 0 0>, rot=<0 -0.0088 0 1>}
 dyn_coat_l_03 {parent=dyn_coat_l_02, loc=<10.5117 0 0>, rot=<0 -0.0104 0 0.9999>}
 dyn_coat_l_04 {parent=dyn_coat_l_03, loc=<15.8782 0 0>, rot=<0 -0.0059 0 1>}
 dyn_coat_l_05 {parent=dyn_coat_l_04, loc=<20.2787 0 0>, rot=<0 0 0 1>}
 dyn_coat_l_06 {parent=dyn_coat_l_05, loc=<18.03 0 0>, rot=<0 0 0 1>}
 dyn_coat_l_07 {parent=dyn_coat_l_06, loc=<15.502 0 0>, rot=<0 0 0 1>}
 dyn_coat_r_MSTR {parent=spine_04, loc=<2.951 10.5522 12.2799>, rot=<-0.0183 -0.418 -0.9056 0.07>}
 dyn_coat_r_01 {parent=dyn_coat_r_MSTR, loc=<0.3 0 0>, rot=<0 0 0 1>}
 dyn_coat_r_02 {parent=dyn_coat_r_01, loc=<11.567 0 0>, rot=<0 -0.0088 0 1>}
 dyn_coat_r_03 {parent=dyn_coat_r_02, loc=<10.5117 0 0>, rot=<0 -0.0104 0 0.9999>}
 dyn_coat_r_04 {parent=dyn_coat_r_03, loc=<15.8782 0 0>, rot=<0 -0.0059 0 1>}
 dyn_coat_r_05 {parent=dyn_coat_r_04, loc=<20.2787 0 0>, rot=<0 0 0 1>}
 dyn_coat_r_06 {parent=dyn_coat_r_05, loc=<18.03 0 0>, rot=<0 0 0 1>}
 dyn_coat_r_07 {parent=dyn_coat_r_06, loc=<15.502 0 0>, rot=<0 0 0 1>}
 dyn_coat_bk_MSTR {parent=spine_02, loc=<-1.2833 -11.6719 0>, rot=<0.0648 -0.7041 0.7041 -0.0648>}
 dyn_coat_bk_01 {parent=dyn_coat_bk_MSTR, loc=<0.3 0 0>, rot=<0 0 0 1>}
 dyn_coat_bk_02 {parent=dyn_coat_bk_01, loc=<10.4809 0 0>, rot=<0 -0.0144 0 -0.9999>}
 dyn_coat_bk_04 {parent=dyn_coat_bk_02, loc=<12.8611 0 0.2968>, rot=<0 0.029 0 -0.9996>}
 dyn_coat_bk_05 {parent=dyn_coat_bk_04, loc=<15.1935 0 -0.4447>, rot=<0 -0.0025 0 -1>}
 dyn_coat_bk_06 {parent=dyn_coat_bk_05, loc=<13.982 0 0>, rot=<0 0 0 1>}
 dyn_coat_bk_07 {parent=dyn_coat_bk_06, loc=<17.1733 0 0>, rot=<0 0 0 1>}
 dyn_coat_bk_l_MSTR {parent=spine_02, loc=<-0.6985 -8.0441 -11.7202>, rot=<0.0903 -0.8963 0.413 -0.1338>}
 dyn_coat_bk_l_01 {parent=dyn_coat_bk_l_MSTR, loc=<0.3 0 0>, rot=<0 -0.049 0 -0.9988>}
 dyn_coat_bk_l_02 {parent=dyn_coat_bk_l_01, loc=<11.0847 0 0>, rot=<0 -0.0378 0 -0.9993>}
 dyn_coat_bk_l_04 {parent=dyn_coat_bk_l_02, loc=<14.588 0 0.0559>, rot=<0 -0.0116 0 -0.9999>}
 dyn_coat_bk_l_05 {parent=dyn_coat_bk_l_04, loc=<13.6078 0 0>, rot=<0 0.036 0 -0.9994>}
 dyn_coat_bk_l_06 {parent=dyn_coat_bk_l_05, loc=<14.9961 0 0>, rot=<0 0 0 1>}
 dyn_coat_bk_l_07 {parent=dyn_coat_bk_l_06, loc=<15.7571 0 0>, rot=<0 0 0 1>}
 dyn_belt_l_MSTR {parent=dyn_coat_bk_l_MSTR, loc=<-1.6531 2.0427 1.4967>, rot=<-0.4468 -0.0342 0.2093 0.8692>}
 dyn_belt_l_01 {parent=dyn_belt_l_MSTR, loc=<0 0 1.762>, rot=<0.0494 -0.0313 -0.1185 -0.9912>}
 dyn_belt_l_02 {parent=dyn_belt_l_01, loc=<8.8248 0 0>, rot=<0.0154 -0.0688 0.147 -0.9866>}
 dyn_belt_l_03 {parent=dyn_belt_l_02, loc=<9.3967 0 0>, rot=<-0.1413 0.0048 0.0292 -0.9895>}
 dyn_belt_l_04 {parent=dyn_belt_l_03, loc=<11.0233 0 0>, rot=<0 -0.0667 0 -0.9978>}
 dyn_belt_l_05 {parent=dyn_belt_l_04, loc=<10.2928 0 0>, rot=<0 0.0161 0 0.9999>}
 dyn_coat_bk_r_MSTR {parent=spine_02, loc=<-0.6985 -8.0441 11.7202>, rot=<-0.1338 0.413 -0.8963 0.0903>}
 dyn_coat_bk_r_01 {parent=dyn_coat_bk_r_MSTR, loc=<0.3 0 0>, rot=<0 0.049 0 0.9988>}
 dyn_coat_bk_r_02 {parent=dyn_coat_bk_r_01, loc=<11.0847 0 0>, rot=<0 0.0378 0 0.9993>}
 dyn_coat_bk_r_04 {parent=dyn_coat_bk_r_02, loc=<14.588 0 0.0559>, rot=<0 0.0116 0 0.9999>}
 dyn_coat_bk_r_05 {parent=dyn_coat_bk_r_04, loc=<13.6078 0 0>, rot=<0 -0.036 0 0.9994>}
 dyn_coat_bk_r_06 {parent=dyn_coat_bk_r_05, loc=<14.9961 0 0>, rot=<0 0 0 1>}
 dyn_coat_bk_r_07 {parent=dyn_coat_bk_r_06, loc=<15.7571 0 0>, rot=<0 0 0 1>}
 dyn_belt_r_MSTR {parent=dyn_coat_bk_r_MSTR, loc=<-1.383 -1.9374 2.1604>, rot=<-0.6166 -0.0433 0.3924 -0.6811>}
 dyn_belt_r_01 {parent=dyn_belt_r_MSTR, loc=<0 0 1.6574>, rot=<0.1518 0.0829 -0.0711 -0.9824>}
 dyn_belt_r_02 {parent=dyn_belt_r_01, loc=<8.8248 0 0>, rot=<0.0035 0.1977 0.1748 0.9645>}
 dyn_belt_r_03 {parent=dyn_belt_r_02, loc=<10.0617 0.0522 0.0447>, rot=<-0.1402 -0.0386 0.034 0.9888>}
 dyn_belt_r_04 {parent=dyn_belt_r_03, loc=<11.6905 0 -0.0405>, rot=<0 0.0303 0 0.9995>}
 dyn_belt_r_05 {parent=dyn_belt_r_04, loc=<10.961 0 -0.0215>, rot=<0 0.0161 0 0.9999>}
 dyn_belt_bk_MSTR {parent=spine_02, loc=<1.6327 -11.2687 0>, rot=<-0.4916 0.5083 -0.5083 0.4916>}
 dyn_belt_bk {parent=dyn_belt_bk_MSTR, loc=<-8.8214 0 -0.3502>, rot=<0 0 0 1>}
 thigh_l {parent=pelvis, loc=<-2.5693 -0.0109 -7.7501>, rot=<0.0012 0.0572 0.0155 -0.9982>}
 calf_l {parent=thigh_l, loc=<-39.0484 0 0>, rot=<0.0006 -0.0161 -0.0391 -0.9991>}
 foot_l {parent=calf_l, loc=<-42.561 0.1077 0.4292>, rot=<0.0418 -0.0455 0.0343 -0.9975>}
 ball_l {parent=foot_l, loc=<-6.1599 13.8264 0.6806>, rot=<0.0001 -0.0001 -0.7063 -0.7079>}
 calf_twist_02_l {parent=calf_l, loc=<-18.8064 -0.0967 0.0397>, rot=<0.0237 0.0095 -0.0035 -0.9997>}
 calf_twist_01_l {parent=calf_l, loc=<-35.813 -0.0967 0.0397>, rot=<0.0237 0.0095 -0.0035 -0.9997>}
 deform_knee_in_base_l {parent=calf_l, loc=<1.298 -5.2957 -0.0496>, rot=<0 0 0 -1>}
 deform_knee_in_l {parent=deform_knee_in_base_l, loc=<0 0 0>, rot=<0 0 0 1>}
 deform_knee_out_base_l {parent=calf_l, loc=<0 4 0>, rot=<0 0 0 -1>}
 deform_knee_out_l {parent=deform_knee_out_base_l, loc=<0 0 0>, rot=<0 0 0 1>}
 thigh_twist_01_l {parent=thigh_l, loc=<-21.6303 0 0>, rot=<0.0013 0 0 -1>}
 deform_groin_base_l {parent=thigh_l, loc=<-2.1908 8.1142 -1.7442>, rot=<0 0 0 -1>}
 deform_groin_l {parent=deform_groin_base_l, loc=<0 0 0>, rot=<0 0 0 1>}
 thigh_r {parent=pelvis, loc=<-2.5693 -0.0109 7.7501>, rot=<0.0572 -0.0012 0.9982 0.0155>}
 calf_r {parent=thigh_r, loc=<39.0484 0 0>, rot=<0.0006 -0.0161 -0.0391 -0.9991>}
 foot_r {parent=calf_r, loc=<42.561 -0.1077 -0.4292>, rot=<0.0418 -0.0455 0.0343 -0.9975>}
 ball_r {parent=foot_r, loc=<6.1599 -13.8264 -0.6806>, rot=<0.0001 -0.0001 -0.7063 -0.7079>}
 calf_twist_02_r {parent=calf_r, loc=<18.8064 0.0967 -0.0397>, rot=<0.0237 0.0095 -0.0035 -0.9997>}
 calf_twist_01_r {parent=calf_r, loc=<35.813 0.0967 -0.0397>, rot=<0.0237 0.0095 -0.0035 -0.9997>}
 deform_knee_in_base_r {parent=calf_r, loc=<-1.298 5.2957 0.0496>, rot=<0 0 0 -1>}
 deform_knee_in_r {parent=deform_knee_in_base_r, loc=<0 0 0>, rot=<0 0 0 1>}
 deform_knee_out_base_r {parent=calf_r, loc=<0 -4 0>, rot=<0 0 0 -1>}
 deform_knee_out_r {parent=deform_knee_out_base_r, loc=<0 0 0>, rot=<0 0 0 1>}
 thigh_twist_01_r {parent=thigh_r, loc=<21.6303 0 0>, rot=<0.0013 0 0 -1>}
 deform_groin_base_r {parent=thigh_r, loc=<2.1908 -8.1142 1.7442>, rot=<0 0 0 -1>}
 deform_groin_r {parent=deform_groin_base_r, loc=<0 0 0>, rot=<0 0 0 1>}
 deform_glute_base_l {parent=pelvis, loc=<-8.8064 -8.8222 -8.4287>, rot=<0.0275 0.0501 -0.4472 -0.8926>}
 deform_glute_l {parent=deform_glute_base_l, loc=<0 0 0>, rot=<0 0 0 1>}
 deform_glute_base_r {parent=pelvis, loc=<-8.8062 -8.8222 8.4287>, rot=<0.0501 -0.0275 0.8926 -0.4472>}
 deform_glute_r {parent=deform_glute_base_r, loc=<0 0 0>, rot=<0 0 0 1>}
 dyn_ground {parent=root, loc=<0 0 0>, rot=<0 0 0 1>}
]
