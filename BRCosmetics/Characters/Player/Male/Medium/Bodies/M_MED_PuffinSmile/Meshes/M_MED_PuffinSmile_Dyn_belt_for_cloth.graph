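# Physics asset for the dynamic belt on the male medium PuffinSmile player body, named for use with cloth.
# Two five-segment belt capsule chains, left and right, hang from small master bodies through swing- and twist-limited constraints, with looser swing at the root.
# Capsules on the pelvis, lower spine, thighs, calves and selected back coat bones have no constraints, apparently serving as collision volumes.
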PhysicsAsset /BRCosmetics/Characters/Player/Male/Medium/Bodies/M_MED_PuffinSmile/Meshes/M_MED_PuffinSmile_Dyn_belt_for_cloth
bodies {
 calf_l [
  {
   shapeType=Capsule
   length=20.6772
   radius=12.4225
   shapeName=calf_l_capsule
   center=<-44.1542 0.83 -1.0265>
   rotation=<rot 82.1202 178.1258 -170.9083>
  }
  {
   shapeType=Capsule
   length=29.2275
   radius=10.2928
   shapeName=calf_l_capsule
   center=<-20.0091 -1.4557 0.3034>
   rotation=<rot 85.3254 -175.5151 178.9213>
  }
 ]
 calf_r [
  {
   shapeType=Capsule
   length=20.6772
   radius=12.4225
   shapeName=calf_l_capsule
   center=<44.1542 -0.83 1.0265>
   rotation=<rot 82.1202 178.1258 9.0917>
  }
  {
   shapeType=Capsule
   length=29.2275
   radius=10.2928
   shapeName=calf_l_capsule
   center=<20.0091 1.4557 -0.3034>
   rotation=<rot 85.3254 -175.5151 -1.0787>
  }
 ]
 dyn_belt_l_01 [{shapeType=Capsule, length=12.3057, radius=4.9941, shapeName=dyn_coat_l_01_capsule, center=<6.5158 0 5.23>, rotation=<rot 90 0 0>}]
 dyn_belt_l_02 [{shapeType=Capsule, length=11.6741, radius=5.6257, shapeName=dyn_coat_l_01_capsule, center=<6.5158 0 4.8557>, rotation=<rot 90 0 0>}]
 dyn_belt_l_03 [{shapeType=Capsule, length=8.2091, radius=9.0907, shapeName=dyn_coat_l_01_capsule, center=<6.5158 0 4.8557>, rotation=<rot 90 0 0>}]
 dyn_belt_l_04 [{shapeType=Capsule, length=8.8371, radius=8.4627, shapeName=dyn_coat_l_01_capsule, center=<6.5158 0 4.6586>, rotation=<rot 90 0 0>}]
 dyn_belt_l_05 [{shapeType=Capsule, length=8.4463, radius=8.8535, shapeName=dyn_coat_l_01_capsule, center=<6.5158 0 5.0643>, rotation=<rot 90 0 0>}]
 dyn_belt_l_MSTR [{shapeType=Capsule, length=0.505, radius=0.505, shapeName=dyn_coat_l_MSTR_capsule}]
 dyn_belt_r_01 [{shapeType=Capsule, length=12.192, radius=5.1078, shapeName=dyn_coat_l_01_capsule, center=<6.5159 0 5.8785>, rotation=<rot 90 0 0>}]
 dyn_belt_r_02 [{shapeType=Capsule, length=12.192, radius=5.1078, shapeName=dyn_coat_l_01_capsule, center=<6.5158 0 5.0683>, rotation=<rot 90 0 0>}]
 dyn_belt_r_03 [{shapeType=Capsule, length=8.2091, radius=9.0907, shapeName=dyn_coat_l_01_capsule, center=<6.5158 0 5.0683>, rotation=<rot 90 0 0>}]
 dyn_belt_r_04 [{shapeType=Capsule, length=7.7949, radius=9.5049, shapeName=dyn_coat_l_01_capsule, center=<6.5159 0 4.4778>, rotation=<rot 90 0 0>}]
 dyn_belt_r_05 [{shapeType=Capsule, length=6.8934, radius=10.4064, shapeName=dyn_coat_l_01_capsule, center=<6.5159 0 5.3141>, rotation=<rot 90 0 0>}]
 dyn_belt_r_MSTR [{shapeType=Capsule, length=0.505, radius=0.505, shapeName=dyn_coat_l_MSTR_capsule}]
 dyn_coat_bk_l_02 [{shapeType=Capsule, length=23.217, radius=13.2579, shapeName=dyn_coat_l_01_capsule, center=<12.5735 0.6962 -11.604>, rotation=<rot 90 0 0>}]
 dyn_coat_bk_l_04 [{
   shapeType=Capsule
   length=21.4567
   radius=15.0182
   shapeName=dyn_coat_l_01_capsule
   center=<11.6466 2.4301 -14.1639>
   rotation=<rot 90 -3.0197 -11.0332>
  }]
 dyn_coat_bk_l_05 [{shapeType=Capsule, length=19.4749, radius=17, shapeName=dyn_coat_l_01_capsule, center=<9.99 1.3969 -15.8015>, rotation=<rot 90 0 0>}]
 dyn_coat_bk_r_02 [{shapeType=Capsule, length=23.217, radius=13.2579, shapeName=dyn_coat_l_01_capsule, center=<12.5735 0.6962 -11.604>, rotation=<rot 90 0 0>}]
 dyn_coat_bk_r_04 [{
   shapeType=Capsule
   length=21.4567
   radius=15.0182
   shapeName=dyn_coat_l_01_capsule
   center=<11.6466 2.4301 -14.1639>
   rotation=<rot 90 -3.0197 -11.0332>
  }]
 dyn_coat_bk_r_05 [{shapeType=Capsule, length=19.4749, radius=17, shapeName=dyn_coat_l_01_capsule, center=<9.99 1.3969 -15.8015>, rotation=<rot 90 0 0>}]
 pelvis [{shapeType=Capsule, length=12.0617, radius=10.4987, shapeName=pelvis_capsule, center=<-3.8176 -2.1282 0>}]
 spine_01 [{shapeType=Capsule, length=6.84, radius=10.4987, shapeName=pelvis_capsule, center=<-0.9991 -1.1308 0>}]
 thigh_l [{shapeType=Capsule, length=28.4627, radius=14.1582, shapeName=pelvis_capsule, center=<-19.9619 1.6961 1.5612>, rotation=<rot 90 0 0>}]
 thigh_r [{shapeType=Capsule, length=28.4627, radius=14.1582, shapeName=pelvis_capsule, center=<19.9619 -1.6961 -1.5612>, rotation=<rot 90 0 0>}]
}
constraints [
 {bone1Name=dyn_belt_r_01, bone2Name=dyn_belt_r_MSTR, swing1LimitDegrees=35, swing2LimitDegrees=45, twistLimitDegrees=5}
 {bone1Name=dyn_belt_r_02, bone2Name=dyn_belt_r_01, swing1LimitDegrees=20, swing2LimitDegrees=45, twistLimitDegrees=5}
 {bone1Name=dyn_belt_r_03, bone2Name=dyn_belt_r_02, swing1LimitDegrees=20, swing2LimitDegrees=45, twistLimitDegrees=5}
 {bone1Name=dyn_belt_r_04, bone2Name=dyn_belt_r_03, swing1LimitDegrees=20, swing2LimitDegrees=45, twistLimitDegrees=5}
 {bone1Name=dyn_belt_l_01, bone2Name=dyn_belt_l_MSTR, swing1LimitDegrees=35, swing2LimitDegrees=45, twistLimitDegrees=5}
 {bone1Name=dyn_belt_l_02, bone2Name=dyn_belt_l_01, swing1LimitDegrees=20, swing2LimitDegrees=45, twistLimitDegrees=5}
 {bone1Name=dyn_belt_l_03, bone2Name=dyn_belt_l_02, swing1LimitDegrees=20, swing2LimitDegrees=45, twistLimitDegrees=5}
 {bone1Name=dyn_belt_l_04, bone2Name=dyn_belt_l_03, swing1LimitDegrees=20, swing2LimitDegrees=45, twistLimitDegrees=5}
 {bone1Name=dyn_belt_l_05, bone2Name=dyn_belt_l_04, swing1LimitDegrees=20, swing2LimitDegrees=45, twistLimitDegrees=5}
 {bone1Name=dyn_belt_r_05, bone2Name=dyn_belt_r_04, swing1LimitDegrees=20, swing2LimitDegrees=45, twistLimitDegrees=5}
]
