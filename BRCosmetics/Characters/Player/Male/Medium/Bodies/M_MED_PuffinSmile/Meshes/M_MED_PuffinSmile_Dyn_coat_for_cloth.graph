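# Physics asset for the PuffinSmile medium male body's dynamic coat, named for use with cloth.
# Left and right coat chains each link a tiny master capsule to three large capsules through tightly angled swing and twist constraints.
# Both masters constrain to spine_04, which has no body here, with free swing and twist.
# Capsules on the pelvis, lower three spine bones and thighs form unconstrained body volumes around the coat.
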PhysicsAsset /BRCosmetics/Characters/Player/Male/Medium/Bodies/M_MED_PuffinSmile/Meshes/M_MED_PuffinSmile_Dyn_coat_for_cloth
bodies {
 dyn_coat_l_01 [{shapeType=Capsule, length=10.7561, radius=8, shapeName=dyn_coat_l_01_capsule, center=<4.667 0 6.9674>, rotation=<rot 90 0 0>}]
 dyn_coat_l_02 [{shapeType=Capsule, length=12.2561, radius=8, shapeName=dyn_coat_l_01_capsule, center=<5.44 0 6.8974>, rotation=<rot 90 0 0>}]
 dyn_coat_l_03 [{shapeType=Capsule, length=7.666, radius=8, shapeName=dyn_coat_l_01_capsule, center=<5.3327 0 8.5462>, rotation=<rot 90 0 0>}]
 dyn_coat_l_MSTR [{shapeType=Capsule, length=0.505, radius=0.505, shapeName=dyn_coat_l_MSTR_capsule}]
 dyn_coat_r_01 [{shapeType=Capsule, length=10.7561, radius=8, shapeName=dyn_coat_l_01_capsule, center=<4.667 0 6.9674>, rotation=<rot 90 0 0>}]
 dyn_coat_r_02 [{shapeType=Capsule, length=12.2561, radius=8, shapeName=dyn_coat_l_01_capsule, center=<5.44 0 6.8974>, rotation=<rot 90 0 0>}]
 dyn_coat_r_03 [{shapeType=Capsule, length=7.666, radius=8, shapeName=dyn_coat_l_01_capsule, center=<5.3327 0 8.5462>, rotation=<rot 90 0 0>}]
 dyn_coat_r_MSTR [{shapeType=Capsule, length=0.505, radius=0.505, shapeName=dyn_coat_l_MSTR_capsule}]
 pelvis [{shapeType=Capsule, length=6.967, radius=11.521, shapeName=pelvis_capsule, center=<-4.8171 -0.6518 -0.0005>}]
 spine_01 [{shapeType=Capsule, length=6.4652, radius=10.9987, shapeName=pelvis_capsule, center=<-0.9991 -0.3721 0>}]
 spine_02 [{shapeType=Capsule, length=7.4401, radius=10.4987, shapeName=pelvis_capsule, center=<0.1451 -0.7547 0>}]
 spine_03 [{shapeType=Capsule, length=7.4401, radius=10.4987, shapeName=pelvis_capsule, center=<2.1769 0.258 0>}]
 thigh_l [{shapeType=Capsule, length=28.841, radius=11.8021, shapeName=pelvis_capsule, center=<-19.998 -0.2298 2.4342>, rotation=<rot 86.3502 180 180>}]
 thigh_r [{shapeType=Capsule, length=28.841, radius=11.8021, shapeName=pelvis_capsule, center=<19.998 0.2298 -2.4342>, rotation=<rot 86.3502 180 180>}]
}
constraints [
 {bone1Name=dyn_coat_l_01, bone2Name=dyn_coat_l_MSTR, swing1LimitDegrees=10, swing2LimitDegrees=10, twistLimitDegrees=5}
 {bone1Name=dyn_coat_l_02, bone2Name=dyn_coat_l_01, swing1LimitDegrees=10, swing2LimitDegrees=10, twistLimitDegrees=5}
 {bone1Name=dyn_coat_l_03, bone2Name=dyn_coat_l_02, swing1LimitDegrees=10, swing2LimitDegrees=10, twistLimitDegrees=5}
 {bone1Name=dyn_coat_r_01, bone2Name=dyn_coat_r_MSTR, swing1LimitDegrees=10, swing2LimitDegrees=10, twistLimitDegrees=5}
 {bone1Name=dyn_coat_r_02, bone2Name=dyn_coat_r_01, swing1LimitDegrees=10, swing2LimitDegrees=10, twistLimitDegrees=5}
 {bone1Name=dyn_coat_r_03, bone2Name=dyn_coat_r_02, swing1LimitDegrees=10, swing2LimitDegrees=10, twistLimitDegrees=5}
 {
  bone1Name=dyn_coat_l_MSTR
  bone2Name=spine_04
  swing1LimitDegrees=45
  swing1Motion=Free
  swing2LimitDegrees=45
  swing2Motion=Free
  twistLimitDegrees=45
  twistMotion=Free
 }
 {
  bone1Name=dyn_coat_r_MSTR
  bone2Name=spine_04
  swing1LimitDegrees=45
  swing1Motion=Free
  swing2LimitDegrees=45
  swing2Motion=Free
  twistLimitDegrees=45
  twistMotion=Free
 }
]
