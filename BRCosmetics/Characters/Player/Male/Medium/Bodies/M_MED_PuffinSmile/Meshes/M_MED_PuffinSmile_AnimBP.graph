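# Animation Blueprint for the PuffinSmile medium male body cosmetic, adding secondary motion to its coat, belts, sleeves and collar.
# It copies the parent mesh pose, runs a Control Rig, lowers dyn_ground on grind rails, then applies AnimDynamics into a shared cache.
# Output selects RigidBody simulation when enabled, otherwise a posed DBNO branch or trail controllers weighted by backward movement.
# Weighted constraints anchor the belts to spine, pelvis and coat bones.
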AnimBlueprint /BRCosmetics/Characters/Player/Male/Medium/Bodies/M_MED_PuffinSmile/Meshes/M_MED_PuffinSmile_AnimBP
target_skeleton=&/M_MED_PuffinSmile_Skeleton
graphs [
 AnimGraph {nodes [
   n1 = "Transform (Modify) Bone - Bone: dyn_coat_l_03"(Alpha=1, Translation="0,0,0", Rotation="0,0,0", Scale="1,1,1") {rotationMode=BMM_Additive, rotationSpace=BCS_BoneSpace, alphaBoolBlend {blendOption=Linear}, boneToModify {boneName=dyn_coat_l_03}}
   n2 = "Transform (Modify) Bone - Bone: dyn_coat_bk_01"(ComponentPose=@n1, Alpha=1, Translation="0,0,0", Rotation="0,0,0", Scale="1,1,1") {rotationMode=BMM_Additive, rotationSpace=BCS_BoneSpace, alphaBoolBlend {blendOption=Linear}, boneToModify {boneName=dyn_coat_bk_01}}
   "Transform (Modify) Bone - Bone: dyn_coat_r_03"(ComponentPose=@n2, Alpha=1, Translation="0,0,0", Rotation="0,0,0", Scale="1,1,1") {rotationMode=BMM_Additive, rotationSpace=BCS_BoneSpace, alphaBoolBlend {blendOption=Linear}, boneToModify {boneName=dyn_coat_r_03}}
   n3 = Use cached pose 'trailCache'()
   n4 = Use cached pose 'dbnoCache'()
   n5 = Get bIsDBNO()
   Output Pose(Result=Blend Poses by bool(
     BlendPose_0=Use cached pose 'RbanCache'()
     BlendPose_1=Blend Poses by bool(BlendPose_0=@n4, BlendPose_1=@n3, bActiveValue=@n5) {blendType=Linear}
     bActiveValue=Get bRigidBodyEnabled()
    ) {blendType=Linear})
   n6 = Control Rig(Source=Copy Pose From Mesh() {bUseAttachedParent=true}) {
    alphaBoolBlend {blendOption=Linear}
    controlRigAssetReference {blueprintRigClass=&/M_MED_PuffinSmile_CtrlRig.M_MED_PuffinSmile_CtrlRig_C}
   }
   n7 = "Transform (Modify) Bone - Bone: dyn_ground"!(
    ComponentPose=Local To Component(LocalPose=@n6)
    bAlphaBoolEnabled:bool=Get bIsOnGrindRail()
    Translation="0,0,-100"
    Rotation="0,0,0"
    Scale="1,1,1"
   ) -> (Pose) {
    alphaInputType=Bool
    translationMode=BMM_Additive
    translationSpace=BCS_BoneSpace
    alphaBoolBlend {blendOption=Linear}
    boneToModify {boneName=dyn_ground}
    translation {z=-100}
   }
   n8 = Anim Dynamics - Bone: dyn_sleeve_r_01(ComponentPose=@n7, Alpha=1, ExternalForce="0,0,0") {
    angularDampingOverride=1
    angularSpringConstant=20000
    bAngularSpring=true
    bLinearSpring=true
    bOverrideAngularDamping=true
    bOverrideLinearDamping=true
    bUseGravityOverride=true
    gravityScale=0.35
    linearDampingOverride=1
    linearSpringConstant=1
    simulationSpace=RootRelative
    alphaBoolBlend {blendOption=Linear}
    boundBone {boneName=dyn_sleeve_r_01}
    componentAppliedLinearAccClamp=<1500 1500 1500>
    componentLinearAccScale=<1 1 1>
    componentLinearVelScale=<1 1 1>
    gravityOverride {z=-980}
    physicsBodyDefinitions [{
      boundBone {boneName=dyn_sleeve_r_01}
      boxExtents=<46.3771 28.0748 11.1329>
      constraintSetup {angularLimitsMax=<5 1 1>, angularLimitsMin=<-5 -1 -1>, angularTarget=<1 0 0>, linearAxesMax=<3.5 0 0>}
      localJointOffset=<-5 0 0>
     }]
   }
   n9 = Anim Dynamics - Bone: dyn_sleeve_r_02(ComponentPose=@n8, Alpha=1, ExternalForce="0,0,0") {
    angularDampingOverride=1
    angularSpringConstant=20000
    bAngularSpring=true
    bLinearSpring=true
    bOverrideAngularDamping=true
    bOverrideLinearDamping=true
    bUseGravityOverride=true
    gravityScale=0.35
    linearDampingOverride=1
    linearSpringConstant=1
    simulationSpace=RootRelative
    alphaBoolBlend {blendOption=Linear}
    boundBone {boneName=dyn_sleeve_r_02}
    componentAppliedLinearAccClamp=<1500 1500 1500>
    componentLinearAccScale=<1 1 1>
    componentLinearVelScale=<1 1 1>
    gravityOverride {z=-980}
    physicsBodyDefinitions [{
      boundBone {boneName=dyn_sleeve_r_02}
      boxExtents=<46.3771 28.0748 11.1329>
      constraintSetup {angularLimitsMax=<5 1 1>, angularLimitsMin=<-5 -1 -1>, angularTarget=<1 0 0>, linearAxesMax=<7 0 0>}
      localJointOffset=<-5 0 0>
     }]
   }
   n10 = Anim Dynamics - Bone: dyn_sleeve_l_01(ComponentPose=@n9, Alpha=1, ExternalForce="0,0,0") {
    angularDampingOverride=1
    angularSpringConstant=20000
    bAngularSpring=true
    bLinearSpring=true
    bOverrideAngularDamping=true
    bOverrideLinearDamping=true
    bUseGravityOverride=true
    gravityScale=0.35
    linearDampingOverride=1
    linearSpringConstant=1
    simulationSpace=RootRelative
    alphaBoolBlend {blendOption=Linear}
    boundBone {boneName=dyn_sleeve_l_01}
    componentAppliedLinearAccClamp=<1500 1500 1500>
    componentLinearAccScale=<1 1 1>
    componentLinearVelScale=<1 1 1>
    gravityOverride {z=-980}
    physicsBodyDefinitions [{
      boundBone {boneName=dyn_sleeve_l_01}
      boxExtents=<46.3771 28.0748 11.1329>
      constraintSetup {angularLimitsMax=<5 1 1>, angularLimitsMin=<-5 -1 -1>, angularTarget=<1 0 0>, linearAxesMin=<-3.5 0 0>}
      localJointOffset=<5 0 0>
     }]
   }
   n11 = Anim Dynamics - Bone: dyn_sleeve_l_02(ComponentPose=@n10, Alpha=1, ExternalForce="0,0,0") {
    angularDampingOverride=1
    angularSpringConstant=20000
    bAngularSpring=true
    bLinearSpring=true
    bOverrideAngularDamping=true
    bOverrideLinearDamping=true
    bUseGravityOverride=true
    gravityScale=0.35
    linearDampingOverride=1
    linearSpringConstant=1
    simulationSpace=RootRelative
    alphaBoolBlend {blendOption=Linear}
    boundBone {boneName=dyn_sleeve_l_02}
    componentAppliedLinearAccClamp=<1500 1500 1500>
    componentLinearAccScale=<1 1 1>
    componentLinearVelScale=<1 1 1>
    gravityOverride {z=-980}
    physicsBodyDefinitions [{
      boundBone {boneName=dyn_sleeve_l_02}
      boxExtents=<46.3771 28.0748 11.1329>
      constraintSetup {angularLimitsMax=<5 1 1>, angularLimitsMin=<-5 -1 -1>, angularTarget=<1 0 0>, linearAxesMin=<-7 0 0>}
      localJointOffset=<5 0 0>
     }]
   }
   n12 = Anim Dynamics - Bone: dyn_lapel_l(ComponentPose=@n11, Alpha=1, ExternalForce="0,0,0") {
    angularDampingOverride=0.9999
    angularSpringConstant=20000
    bAngularSpring=true
    bOverrideAngularDamping=true
    bUseGravityOverride=true
    gravityScale=0.35
    simulationSpace=RootRelative
    alphaBoolBlend {blendOption=Linear}
    boundBone {boneName=dyn_lapel_l}
    componentAppliedLinearAccClamp=<1500 1500 1500>
    componentLinearAccScale=<1 1 1>
    componentLinearVelScale=<1 1 1>
    gravityOverride {z=-980}
    physicsBodyDefinitions [{
      boundBone {boneName=dyn_lapel_l}
      boxExtents=<33.6901 30.8541 11.1329>
      constraintSetup {angularLimitsMax=<1 5 0>, angularLimitsMin=<-1 -5 0>, angularTarget=<1 0 0>}
      localJointOffset=<2.6723 0 -0.0021>
     }]
   }
   n13 = Anim Dynamics - Bone: dyn_lapel_r(ComponentPose=@n12, Alpha=1, ExternalForce="0,0,0") {
    angularDampingOverride=0.9999
    angularSpringConstant=20000
    bAngularSpring=true
    bOverrideAngularDamping=true
    bUseGravityOverride=true
    gravityScale=0.35
    simulationSpace=RootRelative
    alphaBoolBlend {blendOption=Linear}
    boundBone {boneName=dyn_lapel_r}
    componentAppliedLinearAccClamp=<1500 1500 1500>
    componentLinearAccScale=<1 1 1>
    componentLinearVelScale=<1 1 1>
    gravityOverride {z=-980}
    physicsBodyDefinitions [{
      boundBone {boneName=dyn_lapel_r}
      boxExtents=<33.6901 30.8541 11.1329>
      constraintSetup {angularLimitsMax=<1 5 0>, angularLimitsMin=<-1 -5 0>, angularTarget=<1 0 0>}
      localJointOffset=<2.6723 0 -0.0021>
     }]
   }
   n14 = Anim Dynamics - Bone: dyn_belt_bk(ComponentPose=@n13, Alpha=1, ExternalForce="0,0,0") {
    angularDampingOverride=0.9999
    angularSpringConstant=20000
    bAngularSpring=true
    bOverrideAngularDamping=true
    bUseGravityOverride=true
    gravityScale=0.35
    simulationSpace=RootRelative
    alphaBoolBlend {blendOption=Linear}
    boundBone {boneName=dyn_belt_bk}
    componentAppliedLinearAccClamp=<1500 1500 1500>
    componentLinearAccScale=<0.2 0.2 0.2>
    componentLinearVelScale=<5 5 5>
    gravityOverride {z=-980}
    physicsBodyDefinitions [{
      boundBone {boneName=dyn_belt_bk}
      boxExtents=<33.6901 30.8541 11.1329>
      constraintSetup {angularLimitsMax=<1 5 0>, angularLimitsMin=<-1 -15 0>, angularTarget=<1 0 0>}
      localJointOffset=<10.3756 -0.0735 0.0532>
     }]
   }
   n15 = Anim Dynamics - Bone: dyn_collar_fr_l_01(ComponentPose=@n14, Alpha=1, ExternalForce="0,0,0") {
    angularDampingOverride=0.9999
    angularSpringConstant=20000
    bAngularSpring=true
    bOverrideAngularDamping=true
    bUseGravityOverride=true
    gravityScale=0.35
    simulationSpace=RootRelative
    alphaBoolBlend {blendOption=Linear}
    boundBone {boneName=dyn_collar_fr_l_01}
    componentAppliedLinearAccClamp=<1500 1500 1500>
    componentLinearAccScale=<1 1 1>
    componentLinearVelScale=<5 5 5>
    gravityOverride {z=-980}
    physicsBodyDefinitions [{
      boundBone {boneName=dyn_collar_fr_l_01}
      boxExtents=<33.6901 30.8541 11.1329>
      constraintSetup {angularLimitsMax=<1 10 0>, angularLimitsMin=<-1 -3 0>, angularTarget=<1 0 0>}
      localJointOffset=<3 0 -0.0021>
     }]
   }
   n16 = Anim Dynamics - Bone: dyn_collar_fr_l_02(ComponentPose=@n15, Alpha=1, ExternalForce="0,0,0") {
    angularDampingOverride=0.9999
    angularSpringConstant=20000
    bAngularSpring=true
    bOverrideAngularDamping=true
    bUseGravityOverride=true
    gravityScale=0.35
    simulationSpace=RootRelative
    alphaBoolBlend {blendOption=Linear}
    boundBone {boneName=dyn_collar_fr_l_02}
    componentAppliedLinearAccClamp=<1500 1500 1500>
    componentLinearAccScale=<1 1 1>
    componentLinearVelScale=<5 5 5>
    gravityOverride {z=-980}
    physicsBodyDefinitions [{
      boundBone {boneName=dyn_collar_fr_l_02}
      boxExtents=<33.6901 30.8541 11.1329>
      constraintSetup {angularLimitsMax=<1 10 0>, angularLimitsMin=<-1 -3 0>, angularTarget=<1 0 0>}
      localJointOffset=<3 0 -0.0021>
     }]
   }
   n17 = Anim Dynamics - Bone: dyn_collar_mid_l_01(ComponentPose=@n16, Alpha=1, ExternalForce="0,0,0") {
    angularDampingOverride=0.9999
    angularSpringConstant=20000
    bAngularSpring=true
    bOverrideAngularDamping=true
    bUseGravityOverride=true
    gravityScale=0.35
    simulationSpace=RootRelative
    alphaBoolBlend {blendOption=Linear}
    boundBone {boneName=dyn_collar_mid_l_01}
    componentAppliedLinearAccClamp=<1500 1500 1500>
    componentLinearAccScale=<1 1 1>
    componentLinearVelScale=<5 5 5>
    gravityOverride {z=-980}
    physicsBodyDefinitions [{
      boundBone {boneName=dyn_collar_mid_l_01}
      boxExtents=<33.6901 30.8541 11.1329>
      constraintSetup {angularLimitsMax=<1 10 0>, angularLimitsMin=<-1 -3 0>, angularTarget=<1 0 0>}
      localJointOffset=<3 0 -0.0021>
     }]
   }
   n18 = Anim Dynamics - Bone: dyn_collar_mid_l_02(ComponentPose=@n17, Alpha=1, ExternalForce="0,0,0") {
    angularDampingOverride=0.9999
    angularSpringConstant=20000
    bAngularSpring=true
    bOverrideAngularDamping=true
    bUseGravityOverride=true
    gravityScale=0.35
    simulationSpace=RootRelative
    alphaBoolBlend {blendOption=Linear}
    boundBone {boneName=dyn_collar_mid_l_02}
    componentAppliedLinearAccClamp=<1500 1500 1500>
    componentLinearAccScale=<1 1 1>
    componentLinearVelScale=<5 5 5>
    gravityOverride {z=-980}
    physicsBodyDefinitions [{
      boundBone {boneName=dyn_collar_mid_l_02}
      boxExtents=<33.6901 30.8541 11.1329>
      constraintSetup {angularLimitsMax=<1 10 0>, angularLimitsMin=<-1 -3 0>, angularTarget=<1 0 0>}
      localJointOffset=<3 0 -0.0021>
     }]
   }
   n19 = Anim Dynamics - Bone: dyn_collar_bk_l_01(ComponentPose=@n18, Alpha=1, ExternalForce="0,0,0") {
    angularDampingOverride=0.9999
    angularSpringConstant=20000
    bAngularSpring=true
    bOverrideAngularDamping=true
    bUseGravityOverride=true
    gravityScale=0.35
    simulationSpace=RootRelative
    alphaBoolBlend {blendOption=Linear}
    boundBone {boneName=dyn_collar_bk_l_01}
    componentAppliedLinearAccClamp=<1500 1500 1500>
    componentLinearAccScale=<1 1 1>
    componentLinearVelScale=<5 5 5>
    gravityOverride {z=-980}
    physicsBodyDefinitions [{
      boundBone {boneName=dyn_collar_bk_l_01}
      boxExtents=<33.6901 30.8541 11.1329>
      constraintSetup {angularLimitsMax=<1 10 0>, angularLimitsMin=<-1 -3 0>, angularTarget=<1 0 0>}
      localJointOffset=<3 0 -0.0021>
     }]
   }
   n20 = Anim Dynamics - Bone: dyn_collar_bk_l_02(ComponentPose=@n19, Alpha=1, ExternalForce="0,0,0") {
    angularDampingOverride=0.9999
    angularSpringConstant=20000
    bAngularSpring=true
    bOverrideAngularDamping=true
    bUseGravityOverride=true
    gravityScale=0.35
    simulationSpace=RootRelative
    alphaBoolBlend {blendOption=Linear}
    boundBone {boneName=dyn_collar_bk_l_02}
    componentAppliedLinearAccClamp=<1500 1500 1500>
    componentLinearAccScale=<1 1 1>
    componentLinearVelScale=<5 5 5>
    gravityOverride {z=-980}
    physicsBodyDefinitions [{
      boundBone {boneName=dyn_collar_bk_l_02}
      boxExtents=<33.6901 30.8541 11.1329>
      constraintSetup {angularLimitsMax=<1 10 0>, angularLimitsMin=<-1 -3 0>, angularTarget=<1 0 0>}
      localJointOffset=<3 0 -0.0021>
     }]
   }
   n21 = Anim Dynamics - Bone: dyn_collar_bk_01(ComponentPose=@n20, Alpha=1, ExternalForce="0,0,0") {
    angularDampingOverride=0.9999
    angularSpringConstant=20000
    bAngularSpring=true
    bOverrideAngularDamping=true
    bUseGravityOverride=true
    gravityScale=0.35
    simulationSpace=RootRelative
    alphaBoolBlend {blendOption=Linear}
    boundBone {boneName=dyn_collar_bk_01}
    componentAppliedLinearAccClamp=<1500 1500 1500>
    componentLinearAccScale=<1 1 1>
    componentLinearVelScale=<5 5 5>
    gravityOverride {z=-980}
    physicsBodyDefinitions [{
      boundBone {boneName=dyn_collar_bk_01}
      boxExtents=<33.6901 30.8541 11.1329>
      constraintSetup {angularLimitsMax=<1 10 0>, angularLimitsMin=<-1 -3 0>, angularTarget=<1 0 0>}
      localJointOffset=<3 0 -0.0021>
     }]
   }
   n22 = Anim Dynamics - Bone: dyn_collar_bk_02(ComponentPose=@n21, Alpha=1, ExternalForce="0,0,0") {
    angularDampingOverride=0.9999
    angularSpringConstant=20000
    bAngularSpring=true
    bOverrideAngularDamping=true
    bUseGravityOverride=true
    gravityScale=0.35
    simulationSpace=RootRelative
    alphaBoolBlend {blendOption=Linear}
    boundBone {boneName=dyn_collar_bk_02}
    componentAppliedLinearAccClamp=<1500 1500 1500>
    componentLinearAccScale=<1 1 1>
    componentLinearVelScale=<5 5 5>
    gravityOverride {z=-980}
    physicsBodyDefinitions [{
      boundBone {boneName=dyn_collar_bk_02}
      boxExtents=<33.6901 30.8541 11.1329>
      constraintSetup {angularLimitsMax=<1 10 0>, angularLimitsMin=<-1 -3 0>, angularTarget=<1 0 0>}
      localJointOffset=<3 0 -0.0021>
     }]
   }
   n23 = Anim Dynamics - Bone: dyn_collar_bk_r_01(ComponentPose=@n22, Alpha=1, ExternalForce="0,0,0") {
    angularDampingOverride=0.9999
    angularSpringConstant=20000
    bAngularSpring=true
    bOverrideAngularDamping=true
    bUseGravityOverride=true
    gravityScale=0.35
    simulationSpace=RootRelative
    alphaBoolBlend {blendOption=Linear}
    boundBone {boneName=dyn_collar_bk_r_01}
    componentAppliedLinearAccClamp=<1500 1500 1500>
    componentLinearAccScale=<1 1 1>
    componentLinearVelScale=<5 5 5>
    gravityOverride {z=-980}
    physicsBodyDefinitions [{
      boundBone {boneName=dyn_collar_bk_r_01}
      boxExtents=<33.6901 30.8541 11.1329>
      constraintSetup {angularLimitsMax=<1 10 0>, angularLimitsMin=<-1 -3 0>, angularTarget=<1 0 0>}
      localJointOffset=<3 0 -0.0021>
     }]
   }
   n24 = Anim Dynamics - Bone: dyn_collar_bk_r_02(ComponentPose=@n23, Alpha=1, ExternalForce="0,0,0") {
    angularDampingOverride=0.9999
    angularSpringConstant=20000
    bAngularSpring=true
    bOverrideAngularDamping=true
    bUseGravityOverride=true
    gravityScale=0.35
    simulationSpace=RootRelative
    alphaBoolBlend {blendOption=Linear}
    boundBone {boneName=dyn_collar_bk_r_02}
    componentAppliedLinearAccClamp=<1500 1500 1500>
    componentLinearAccScale=<1 1 1>
    componentLinearVelScale=<5 5 5>
    gravityOverride {z=-980}
    physicsBodyDefinitions [{
      boundBone {boneName=dyn_collar_bk_r_02}
      boxExtents=<33.6901 30.8541 11.1329>
      constraintSetup {angularLimitsMax=<1 10 0>, angularLimitsMin=<-1 -3 0>, angularTarget=<1 0 0>}
      localJointOffset=<3 0 -0.0021>
     }]
   }
   n25 = Anim Dynamics - Bone: dyn_collar_mid_r_01(ComponentPose=@n24, Alpha=1, ExternalForce="0,0,0") {
    angularDampingOverride=0.9999
    angularSpringConstant=20000
    bAngularSpring=true
    bOverrideAngularDamping=true
    bUseGravityOverride=true
    gravityScale=0.35
    simulationSpace=RootRelative
    alphaBoolBlend {blendOption=Linear}
    boundBone {boneName=dyn_collar_mid_r_01}
    componentAppliedLinearAccClamp=<1500 1500 1500>
    componentLinearAccScale=<1 1 1>
    componentLinearVelScale=<5 5 5>
    gravityOverride {z=-980}
    physicsBodyDefinitions [{
      boundBone {boneName=dyn_collar_mid_r_01}
      boxExtents=<33.6901 30.8541 11.1329>
      constraintSetup {angularLimitsMax=<1 10 0>, angularLimitsMin=<-1 -3 0>, angularTarget=<1 0 0>}
      localJointOffset=<3 0 -0.0021>
     }]
   }
   n26 = Anim Dynamics - Bone: dyn_collar_mid_r_02(ComponentPose=@n25, Alpha=1, ExternalForce="0,0,0") {
    angularDampingOverride=0.9999
    angularSpringConstant=20000
    bAngularSpring=true
    bOverrideAngularDamping=true
    bUseGravityOverride=true
    gravityScale=0.35
    simulationSpace=RootRelative
    alphaBoolBlend {blendOption=Linear}
    boundBone {boneName=dyn_collar_mid_r_02}
    componentAppliedLinearAccClamp=<1500 1500 1500>
    componentLinearAccScale=<1 1 1>
    componentLinearVelScale=<5 5 5>
    gravityOverride {z=-980}
    physicsBodyDefinitions [{
      boundBone {boneName=dyn_collar_mid_r_02}
      boxExtents=<33.6901 30.8541 11.1329>
      constraintSetup {angularLimitsMax=<1 10 0>, angularLimitsMin=<-1 -3 0>, angularTarget=<1 0 0>}
      localJointOffset=<3 0 -0.0021>
     }]
   }
   n27 = Anim Dynamics - Bone: dyn_collar_fr_r_01(ComponentPose=@n26, Alpha=1, ExternalForce="0,0,0") {
    angularDampingOverride=0.9999
    angularSpringConstant=20000
    bAngularSpring=true
    bOverrideAngularDamping=true
    bUseGravityOverride=true
    gravityScale=0.35
    simulationSpace=RootRelative
    alphaBoolBlend {blendOption=Linear}
    boundBone {boneName=dyn_collar_fr_r_01}
    componentAppliedLinearAccClamp=<1500 1500 1500>
    componentLinearAccScale=<1 1 1>
    componentLinearVelScale=<5 5 5>
    gravityOverride {z=-980}
    physicsBodyDefinitions [{
      boundBone {boneName=dyn_collar_fr_r_01}
      boxExtents=<33.6901 30.8541 11.1329>
      constraintSetup {angularLimitsMax=<1 10 0>, angularLimitsMin=<-1 -3 0>, angularTarget=<1 0 0>}
      localJointOffset=<3 0 -0.0021>
     }]
   }
   n28 = Anim Dynamics - Bone: dyn_collar_fr_r_02(ComponentPose=@n27, Alpha=1, ExternalForce="0,0,0") {
    angularDampingOverride=0.9999
    angularSpringConstant=20000
    bAngularSpring=true
    bOverrideAngularDamping=true
    bUseGravityOverride=true
    gravityScale=0.35
    simulationSpace=RootRelative
    alphaBoolBlend {blendOption=Linear}
    boundBone {boneName=dyn_collar_fr_r_02}
    componentAppliedLinearAccClamp=<1500 1500 1500>
    componentLinearAccScale=<1 1 1>
    componentLinearVelScale=<5 5 5>
    gravityOverride {z=-980}
    physicsBodyDefinitions [{
      boundBone {boneName=dyn_collar_fr_r_02}
      boxExtents=<33.6901 30.8541 11.1329>
      constraintSetup {angularLimitsMax=<1 10 0>, angularLimitsMin=<-1 -3 0>, angularTarget=<1 0 0>}
      localJointOffset=<3 0 -0.0021>
     }]
   }
   Save cached pose 'StartupCache'(Pose=Component To Local(ComponentPose=@n28)) {cacheName=StartupCache}
   n29 = Get BwdMovement() -> (BwdMovement:double)
   n30 = Trail controller - Bone: dyn_coat_l_06(ComponentPose=Local To Component(LocalPose=Use cached pose 'StartupCache'()), Alpha=@n29) {
    bUsePlanarLimit=true
    chainLength=4
    debugLifeTime=0.1
    maxDeltaTime=0.333
    alphaScaleBiasClamp {bInterpResult=true}
    rotationLimits [{limitMax=<180 180 180>}, {limitMax=<180 180 180>}, {limitMax=<180 180 180>}, {limitMax=<180 180 180>}]
    rotationOffsets [<0 0 0>, <0 0 0>, <0 0 0>, <0 0 0>]
    trailBone {boneName=dyn_coat_l_06}
    trailRelaxationSpeed {editorCurveData {keys [<at -0.0134 19.3841 RCIM_Linear>,<at 0.5079 12.8986 RCIM_Linear>,<at 0.9955 9.6044 RCIM_Linear>]}}
   }
   n31 = Trail controller - Bone: dyn_coat_bk_l_06(ComponentPose=@n30, Alpha=@n29) {
    bUsePlanarLimit=true
    chainLength=4
    debugLifeTime=0.1
    maxDeltaTime=0.333
    alphaScaleBiasClamp {bInterpResult=true}
    rotationLimits [{limitMax=<180 180 180>}, {limitMax=<180 180 180>}, {limitMax=<180 180 180>}, {limitMax=<180 180 180>}]
    rotationOffsets [<0 0 0>, <0 0 0>, <0 0 0>, <0 0 0>]
    trailBone {boneName=dyn_coat_bk_l_06}
    trailRelaxationSpeed {editorCurveData {keys [<at -0.0134 19.3841 RCIM_Linear>,<at 0.5079 12.8986 RCIM_Linear>,<at 0.9955 9.6044 RCIM_Linear>]}}
   }
   n32 = Trail controller - Bone: dyn_coat_bk_04(ComponentPose=@n31, Alpha=@n29) {
    bUsePlanarLimit=true
    debugLifeTime=0.1
    maxDeltaTime=0.0333
    alphaScaleBiasClamp {bInterpResult=true}
    rotationLimits [{limitMax=<180 180 180>}, {limitMax=<180 180 180>}]
    rotationOffsets [<0 0 0>, <0 0 0>]
    trailBone {boneName=dyn_coat_bk_04}
    trailRelaxationSpeed {editorCurveData {keys [<at 0 15 RCIM_Linear>,<at 0.5289 13.1838 RCIM_Linear>,<at 1 6 RCIM_Linear>]}}
   }
   n33 = Trail controller - Bone: dyn_coat_bk_r_06(ComponentPose=@n32, Alpha=@n29) {
    bUsePlanarLimit=true
    chainLength=4
    debugLifeTime=0.1
    maxDeltaTime=0.333
    alphaScaleBiasClamp {bInterpResult=true}
    rotationLimits [{limitMax=<180 180 180>}, {limitMax=<180 180 180>}, {limitMax=<180 180 180>}, {limitMax=<180 180 180>}]
    rotationOffsets [<0 0 0>, <0 0 0>, <0 0 0>, <0 0 0>]
    trailBone {boneName=dyn_coat_bk_r_06}
    trailRelaxationSpeed {editorCurveData {keys [<at -0.0134 19.3841 RCIM_Linear>,<at 0.5079 12.8986 RCIM_Linear>,<at 0.9955 9.6044 RCIM_Linear>]}}
   }
   n34 = Trail controller - Bone: dyn_coat_r_06(ComponentPose=@n33, Alpha=@n29) {
    bUsePlanarLimit=true
    chainLength=4
    debugLifeTime=0.1
    maxDeltaTime=0.0333
    alphaScaleBiasClamp {bInterpResult=true}
    rotationLimits [{limitMax=<180 180 180>}, {limitMax=<180 180 180>}, {limitMax=<180 90 180>, limitMin=<-180 -90 -180>}, {limitMax=<180 180 180>}]
    rotationOffsets [<0 0 0>, <0 0 0>, <0 0 0>, <0 0 0>]
    trailBone {boneName=dyn_coat_r_06}
    trailRelaxationSpeed {editorCurveData {keys [<at 0 15 RCIM_Linear>,<at 0.5289 13.1838 RCIM_Linear>,<at 1 6 RCIM_Linear>]}}
   }
   n35 = "Constraint - dyn_belt_r_MSTR (PPPPPPPP)"!(
    ComponentPose=@n34
    Alpha=1
    ConstraintWeights_0=0.365
    ConstraintWeights_1:float=0.3055
    ConstraintWeights_2:float=0.301
    ConstraintWeights_3:float=0.0143
    ConstraintWeights_4:float=0.0115
    ConstraintWeights_5:float=0.001
    ConstraintWeights_6:float=0.0009
    ConstraintWeights_7:float=0.0008
   ) -> (Pose) {
    alphaBoolBlend {blendOption=Linear}
    boneToModify {boneName=dyn_belt_r_MSTR}
    constraintSetup [
     {targetBone {boneName=spine_02}}
     {targetBone {boneName=spine_01}}
     {targetBone {boneName=pelvis}}
     {targetBone {boneName=dyn_coat_bk_r_01}}
     {targetBone {boneName=dyn_coat_r_03}}
     {targetBone {boneName=dyn_coat_bk_r_02}}
     {targetBone {boneName=spine_03}}
     {targetBone {boneName=dyn_coat_r_02}}
    ]
    constraintWeights [1,1,1,1,1,1,1,1]
   }
   n36 = "Constraint - dyn_belt_l_MSTR (PPPPPPPP)"!(
    ComponentPose=@n35
    Alpha=1
    ConstraintWeights_0=0.2744
    ConstraintWeights_1:float=0.2648
    ConstraintWeights_2:float=0.2016
    ConstraintWeights_3:float=0.1424
    ConstraintWeights_4:float=0.0407
    ConstraintWeights_5:float=0.0363
    ConstraintWeights_6:float=0.0264
    ConstraintWeights_7:float=0.0135
   ) -> (Pose) {
    alphaBoolBlend {blendOption=Linear}
    boneToModify {boneName=dyn_belt_l_MSTR}
    constraintSetup [
     {targetBone {boneName=spine_01}}
     {targetBone {boneName=pelvis}}
     {targetBone {boneName=spine_02}}
     {targetBone {boneName=spine_03}}
     {targetBone {boneName=dyn_coat_bk_l_01}}
     {targetBone {boneName=spine_04}}
     {targetBone {boneName=dyn_coat_bk_l_02}}
     {targetBone {boneName=dyn_coat_bk_01}}
    ]
    constraintWeights [1,1,1,1,1,1,1,1]
   }
   Save cached pose 'trailCache'(Pose=Component To Local(ComponentPose=@n36)) {cacheName=trailCache}
   n37 = "Transform (Modify) Bone - Bone: dyn_coat_l_01"(
    ComponentPose=Local To Component(LocalPose=Use cached pose 'StartupCache'())
    Alpha=1
    Translation="0,0,0"
    Rotation="32.145,11.529,-3.902"
    Scale="1,1,1"
   ) {
    rotationMode=BMM_Additive
    rotationSpace=BCS_BoneSpace
    alphaBoolBlend {blendOption=Linear}
    boneToModify {boneName=dyn_coat_l_01}
    rotation=<rot 32.145 11.529 -3.902>
   }
   n38 = "Transform (Modify) Bone - Bone: dyn_coat_l_03"(ComponentPose=@n37, Alpha=1, Translation="0,0,0", Rotation="-0.909,17.433,8.209", Scale="1,1,1") {
    rotationMode=BMM_Additive
    rotationSpace=BCS_BoneSpace
    alphaBoolBlend {blendOption=Linear}
    boneToModify {boneName=dyn_coat_l_03}
    rotation=<rot -0.909 17.433 8.209>
   }
   n39 = "Transform (Modify) Bone - Bone: dyn_coat_l_05"(ComponentPose=@n38, Alpha=1, Translation="0,0,0", Rotation="5.741,-9.962,0.410", Scale="1,1,1") {
    rotationMode=BMM_Additive
    rotationSpace=BCS_BoneSpace
    alphaBoolBlend {blendOption=Linear}
    boneToModify {boneName=dyn_coat_l_05}
    rotation=<rot 5.741 -9.9619 0.41>
   }
   n40 = "Transform (Modify) Bone - Bone: dyn_coat_bk_l_02"(ComponentPose=@n39, Alpha=1, Translation="0,0,0", Rotation="-8.860,3.786,-1.333", Scale="1,1,1") {
    rotationMode=BMM_Additive
    rotationSpace=BCS_BoneSpace
    alphaBoolBlend {blendOption=Linear}
    boneToModify {boneName=dyn_coat_bk_l_02}
    rotation=<rot -8.86 3.786 -1.333>
   }
   n41 = "Transform (Modify) Bone - Bone: dyn_coat_bk_l_04"(ComponentPose=@n40, Alpha=1, Translation="0,0,0", Rotation="-0.703,3.196,7.134", Scale="1,1,1") {
    rotationMode=BMM_Additive
    rotationSpace=BCS_BoneSpace
    alphaBoolBlend {blendOption=Linear}
    boneToModify {boneName=dyn_coat_bk_l_04}
    rotation=<rot -0.703 3.196 7.134>
   }
   n42 = "Transform (Modify) Bone - Bone: dyn_coat_bk_l_05"(ComponentPose=@n41, Alpha=1, Translation="0,0,0", Rotation="-15.765,19.004,-5.367", Scale="1,1,1") {
    rotationMode=BMM_Additive
    rotationSpace=BCS_BoneSpace
    alphaBoolBlend {blendOption=Linear}
    boneToModify {boneName=dyn_coat_bk_l_05}
    rotation=<rot -15.765 19.004 -5.367>
   }
   n43 = "Transform (Modify) Bone - Bone: dyn_coat_bk_01"(ComponentPose=@n42, Alpha=1, Translation="0,0,0", Rotation="-15.781,-0.215,-1.367", Scale="1,1,1") {
    rotationMode=BMM_Additive
    rotationSpace=BCS_BoneSpace
    alphaBoolBlend {blendOption=Linear}
    boneToModify {boneName=dyn_coat_bk_01}
    rotation=<rot -15.781 -0.215 -1.367>
   }
   n44 = "Transform (Modify) Bone - Bone: dyn_coat_bk_02"(ComponentPose=@n43, Alpha=1, Translation="0,0,0", Rotation="-7.246,0.129,0.127", Scale="1,1,1") {
    rotationMode=BMM_Additive
    rotationSpace=BCS_BoneSpace
    alphaBoolBlend {blendOption=Linear}
    boneToModify {boneName=dyn_coat_bk_02}
    rotation=<rot -7.246 0.129 0.127>
   }
   n45 = "Transform (Modify) Bone - Bone: dyn_coat_bk_04"(ComponentPose=@n44, Alpha=1, Translation="0,0,0", Rotation="-10.235,-0.453,0.706", Scale="1,1,1") {
    rotationMode=BMM_Additive
    rotationSpace=BCS_BoneSpace
    alphaBoolBlend {blendOption=Linear}
    boneToModify {boneName=dyn_coat_bk_04}
    rotation=<rot -10.235 -0.453 0.706>
   }
   n46 = "Transform (Modify) Bone - Bone: dyn_coat_bk_05"(ComponentPose=@n45, Alpha=1, Translation="0,0,0", Rotation="-2.311,-0.443,0.100", Scale="1,1,1") {
    rotationMode=BMM_Additive
    rotationSpace=BCS_BoneSpace
    alphaBoolBlend {blendOption=Linear}
    boneToModify {boneName=dyn_coat_bk_05}
    rotation=<rot -2.311 -0.443 0.1>
   }
   n47 = "Transform (Modify) Bone - Bone: dyn_coat_bk_r_02"(ComponentPose=@n46, Alpha=1, Translation="0,0,0", Rotation="-10.873,-8.896,1.662", Scale="1,1,1") {
    rotationMode=BMM_Additive
    rotationSpace=BCS_BoneSpace
    alphaBoolBlend {blendOption=Linear}
    boneToModify {boneName=dyn_coat_bk_r_02}
    rotation=<rot -10.873 -8.896 1.662>
   }
   n48 = "Transform (Modify) Bone - Bone: dyn_coat_bk_r_04"(ComponentPose=@n47, Alpha=1, Translation="0,0,0", Rotation="-7.250,-3.668,-0.463", Scale="1,1,1") {
    rotationMode=BMM_Additive
    rotationSpace=BCS_BoneSpace
    alphaBoolBlend {blendOption=Linear}
    boneToModify {boneName=dyn_coat_bk_r_04}
    rotation=<rot -7.25 -3.668 -0.463>
   }
   n49 = "Transform (Modify) Bone - Bone: dyn_coat_bk_r_05"(ComponentPose=@n48, Alpha=1, Translation="0,0,0", Rotation="-19.163,-11.226,1.333", Scale="1,1,1") {
    rotationMode=BMM_Additive
    rotationSpace=BCS_BoneSpace
    alphaBoolBlend {blendOption=Linear}
    boneToModify {boneName=dyn_coat_bk_r_05}
    rotation=<rot -19.163 -11.226 1.333>
   }
   n50 = "Transform (Modify) Bone - Bone: dyn_coat_r_01"(ComponentPose=@n49, Alpha=1, Translation="0,0,0", Rotation="15.009,1.233,0.350", Scale="1,1,1") {
    rotationMode=BMM_Additive
    rotationSpace=BCS_BoneSpace
    alphaBoolBlend {blendOption=Linear}
    boneToModify {boneName=dyn_coat_r_01}
    rotation=<rot 15.009 1.233 0.35>
   }
   n51 = "Transform (Modify) Bone - Bone: dyn_coat_r_03"(ComponentPose=@n50, Alpha=1, Translation="0,0,0", Rotation="-21.159,-40.049,0.078", Scale="1,1,1") {
    rotationMode=BMM_Additive
    rotationSpace=BCS_BoneSpace
    alphaBoolBlend {blendOption=Linear}
    boneToModify {boneName=dyn_coat_r_03}
    rotation=<rot -21.159 -40.049 0.078>
   }
   n52 = "Transform (Modify) Bone - Bone: dyn_coat_r_05"(ComponentPose=@n51, Alpha=1, Translation="0,0,0", Rotation="0,10,0", Scale="1,1,1") {
    rotationMode=BMM_Additive
    rotationSpace=BCS_BoneSpace
    alphaBoolBlend {blendOption=Linear}
    boneToModify {boneName=dyn_coat_r_05}
    rotation {yaw=10}
   }
   n53 = Look At - Bone: dyn_coat_l_01(ComponentPose=@n52, Alpha=1) {
    bUseLookUpAxis=true
    alphaBoolBlend {blendOption=Linear}
    boneToModify {boneName=dyn_coat_l_01}
    lookAtLocation=<0.0206 -1.6234 -5.8436>
    lookAtTarget {boneReference {boneName=foot_l}}
    lookAt_Axis {axis=<1 0>}
    lookUp_Axis {axis {x=0, z=1}}
   }
   n54 = Look At - Bone: dyn_coat_r_01(ComponentPose=@n53, Alpha=1) {
    bUseLookUpAxis=true
    alphaBoolBlend {blendOption=Linear}
    boneToModify {boneName=dyn_coat_r_01}
    lookAtLocation=<7.3292 23.916 9.591>
    lookAtTarget {boneReference {boneName=foot_r}}
    lookAt_Axis {axis=<1 0>}
    lookUp_Axis {axis {x=0, z=1}}
   }
   n55 = Trail controller - Bone: dyn_coat_l_06!(ComponentPose=@n54, Alpha=1, RelaxationSpeedScale:float=2, FakeVelocity:Vector="0,0,0") -> (Pose) {
    bShowTrailLocation=true
    bUsePlanarLimit=true
    chainLength=3
    debugLifeTime=0.1
    lastBoneRotationAnimAlphaBlend=1
    maxDeltaTime=0.0333
    relaxationSpeedScale=1.2
    alphaScaleBiasClamp {bInterpResult=true, interpSpeedDecreasing=25, interpSpeedIncreasing=25}
    rotationLimits [{limitMax=<180 180 180>}, {limitMax=<180 180 180>}, {limitMax=<180 180 180>}]
    rotationOffsets [<0 0 0>, <0 0 0>, <0 0 0>]
    trailBone {boneName=dyn_coat_l_06}
    trailRelaxationSpeed {editorCurveData {keys [<at 0 15 RCIM_Linear>,<at 1 6 RCIM_Linear>]}}
   }
   n56 = Trail controller - Bone: dyn_coat_bk_l_06!(ComponentPose=@n55, Alpha=1, RelaxationSpeedScale:float=2, FakeVelocity:Vector="0,0,0") -> (Pose) {
    bShowTrailLocation=true
    bUsePlanarLimit=true
    chainLength=4
    debugLifeTime=0.1
    lastBoneRotationAnimAlphaBlend=1
    maxDeltaTime=0.0333
    relaxationSpeedScale=1.2
    alphaScaleBiasClamp {bInterpResult=true, interpSpeedDecreasing=25, interpSpeedIncreasing=25}
    rotationLimits [{limitMax=<180 180 180>}, {limitMax=<180 180 180>}, {limitMax=<180 180 180>}, {limitMax=<180 180 180>}]
    rotationOffsets [<0 0 0>, <0 0 0>, <0 0 0>, <0 0 0>]
    trailBone {boneName=dyn_coat_bk_l_06}
    trailRelaxationSpeed {editorCurveData {keys [<at 0 15 RCIM_Linear>,<at 1 6 RCIM_Linear>]}}
   }
   n57 = Trail controller - Bone: dyn_coat_bk_04!(ComponentPose=@n56, Alpha=1, RelaxationSpeedScale:float=2, FakeVelocity:Vector="0,0,0") -> (Pose) {
    bReorientParentToChild=false
    bShowTrailLocation=true
    bUsePlanarLimit=true
    debugLifeTime=0.1
    lastBoneRotationAnimAlphaBlend=1
    maxDeltaTime=0.0333
    relaxationSpeedScale=1.2
    alphaScaleBiasClamp {bInterpResult=true, interpSpeedDecreasing=25, interpSpeedIncreasing=25}
    rotationLimits [{limitMax=<180 180 180>}, {limitMax=<180 180 180>}]
    rotationOffsets [<0 0 0>, <0 0 0>]
    trailBone {boneName=dyn_coat_bk_04}
    trailRelaxationSpeed {editorCurveData {keys [<at 0 15 RCIM_Linear>,<at 1 6 RCIM_Linear>]}}
   }
   n58 = Trail controller - Bone: dyn_coat_bk_r_06!(ComponentPose=@n57, Alpha=1, RelaxationSpeedScale:float=2, FakeVelocity:Vector="0,0,0") -> (Pose) {
    bShowTrailLocation=true
    bUsePlanarLimit=true
    chainLength=3
    debugLifeTime=0.1
    lastBoneRotationAnimAlphaBlend=1
    maxDeltaTime=0.0333
    relaxationSpeedScale=1.2
    alphaScaleBiasClamp {bInterpResult=true, interpSpeedDecreasing=25, interpSpeedIncreasing=25}
    rotationLimits [{limitMax=<180 180 180>}, {limitMax=<180 180 180>}, {limitMax=<180 180 180>}]
    rotationOffsets [<0 0 0>, <0 0 0>, <0 0 0>]
    trailBone {boneName=dyn_coat_bk_r_06}
    trailRelaxationSpeed {editorCurveData {keys [<at 0 15 RCIM_Linear>,<at 1 6 RCIM_Linear>]}}
   }
   n59 = Trail controller - Bone: dyn_coat_r_06!(ComponentPose=@n58, Alpha=1, RelaxationSpeedScale:float=2, FakeVelocity:Vector="0,0,0") -> (Pose) {
    bShowTrailLocation=true
    bUsePlanarLimit=true
    chainLength=3
    debugLifeTime=0.1
    lastBoneRotationAnimAlphaBlend=1
    maxDeltaTime=0.0333
    relaxationSpeedScale=1.2
    alphaScaleBiasClamp {bInterpResult=true, interpSpeedDecreasing=25, interpSpeedIncreasing=25}
    rotationLimits [{limitMax=<180 180 180>}, {limitMax=<180 180 180>}, {limitMax=<180 180 180>}]
    rotationOffsets [<0 0 0>, <0 0 0>, <0 0 0>]
    trailBone {boneName=dyn_coat_r_06}
    trailRelaxationSpeed {editorCurveData {keys [<at 0 15 RCIM_Linear>,<at 1 6 RCIM_Linear>]}}
   }
   n60 = "Constraint - dyn_belt_r_MSTR (PPPPPPPP)"(
    ComponentPose=@n59
    Alpha=1
    ConstraintWeights_0=0.365
    ConstraintWeights_1=0.3055
    ConstraintWeights_2=0.301
    ConstraintWeights_3=0.0143
    ConstraintWeights_4=0.0115
    ConstraintWeights_5=0.001
    ConstraintWeights_6=0.0009
    ConstraintWeights_7=0.0008
   ) {
    alphaBoolBlend {blendOption=Linear}
    boneToModify {boneName=dyn_belt_r_MSTR}
    constraintSetup [
     {targetBone {boneName=spine_02}}
     {targetBone {boneName=spine_01}}
     {targetBone {boneName=pelvis}}
     {targetBone {boneName=dyn_coat_bk_r_01}}
     {targetBone {boneName=dyn_coat_r_03}}
     {targetBone {boneName=dyn_coat_bk_r_02}}
     {targetBone {boneName=spine_03}}
     {targetBone {boneName=dyn_coat_r_02}}
    ]
    constraintWeights [1,1,1,1,1,1,1,1]
   }
   n61 = "Constraint - dyn_belt_l_MSTR (PPPPPPPP)"(
    ComponentPose=@n60
    Alpha=1
    ConstraintWeights_0=0.2744
    ConstraintWeights_1=0.2648
    ConstraintWeights_2=0.2016
    ConstraintWeights_3=0.1424
    ConstraintWeights_4=0.0407
    ConstraintWeights_5=0.0363
    ConstraintWeights_6=0.0264
    ConstraintWeights_7=0.0135
   ) {
    alphaBoolBlend {blendOption=Linear}
    boneToModify {boneName=dyn_belt_l_MSTR}
    constraintSetup [
     {targetBone {boneName=spine_01}}
     {targetBone {boneName=pelvis}}
     {targetBone {boneName=spine_02}}
     {targetBone {boneName=spine_03}}
     {targetBone {boneName=dyn_coat_bk_l_01}}
     {targetBone {boneName=spine_04}}
     {targetBone {boneName=dyn_coat_bk_l_02}}
     {targetBone {boneName=dyn_coat_bk_01}}
    ]
    constraintWeights [1,1,1,1,1,1,1,1]
   }
   n62 = Look At - Bone: dyn_belt_l_01(ComponentPose=@n61, Alpha=1) {
    bUseLookUpAxis=true
    alphaBoolBlend {blendOption=Linear}
    boneToModify {boneName=dyn_belt_l_01}
    lookAtLocation=<-9.3472 3.6832 15.1208>
    lookAtTarget {boneReference {boneName=dyn_coat_l_05}}
    lookAt_Axis {axis=<1 0>}
    lookUp_Axis {axis {x=0, z=1}}
   }
   Save cached pose 'dbnoCache'(Pose=Component To Local(ComponentPose=@n62)) {cacheName=dbnoCache}
   n63 = Get GravityOverride()
   n64 = RigidBody!(
    ComponentPose=Local To Component(LocalPose=Use cached pose 'StartupCache'())
    Alpha
    OverrideWorldGravity:Vector=@n63
    ExternalForce
    SimSpaceSettings:Sim Space Settings Structure=(WorldAlpha=1,VelocityScaleZ=0,DampingAlpha=1,MaxLinearVelocity=500,MaxAngularVelocity=3,MaxLinearAcceleration=700,MaxAngularAcceleration=1000,ExternalLinearDragV=(X=0.1,Y=0.1,Z=0),ExternalLinearVelocity=(X=0,Y=0,Z=0),ExternalAngularVelocity=(X=0,Y=0,Z=0))
   ) -> (Pose) {
    bClampLinearTranslationLimitToRefPose=true
    bOverrideWorldGravity=true
    overridePhysicsAsset=&/M_MED_PuffinSmile_Dyn_coat
    simulationSpace=BaseBoneSpace
    alphaBoolBlend {blendOption=Linear}
    baseBoneRef {boneName=pelvis}
    overrideWorldGravity {z=-980}
    simSpaceSettings {
     maxAngularAcceleration=1000
     maxAngularVelocity=3
     maxLinearAcceleration=700
     maxLinearVelocity=500
     velocityScaleZ=0
     worldAlpha=1
     externalLinearDragV=<0.1 0.1>
    }
   }
   n65 = "Constraint - dyn_belt_r_MSTR (PPPPPPPP)"(
    ComponentPose=@n64
    Alpha=1
    ConstraintWeights_0=0.365
    ConstraintWeights_1=0.3055
    ConstraintWeights_2=0.301
    ConstraintWeights_3=0.0143
    ConstraintWeights_4=0.0115
    ConstraintWeights_5=0.001
    ConstraintWeights_6=0.0009
    ConstraintWeights_7=0.0008
   ) {
    alphaBoolBlend {blendOption=Linear}
    boneToModify {boneName=dyn_belt_r_MSTR}
    constraintSetup [
     {targetBone {boneName=spine_02}}
     {targetBone {boneName=spine_01}}
     {targetBone {boneName=pelvis}}
     {targetBone {boneName=dyn_coat_bk_r_01}}
     {targetBone {boneName=dyn_coat_r_03}}
     {targetBone {boneName=dyn_coat_bk_r_02}}
     {targetBone {boneName=spine_03}}
     {targetBone {boneName=dyn_coat_r_02}}
    ]
    constraintWeights [1,1,1,1,1,1,1,1]
   }
   n66 = "Constraint - dyn_belt_l_MSTR (PPPPPPPP)"(
    ComponentPose=@n65
    Alpha=1
    ConstraintWeights_0=0.2744
    ConstraintWeights_1=0.2648
    ConstraintWeights_2=0.2016
    ConstraintWeights_3=0.1424
    ConstraintWeights_4=0.0407
    ConstraintWeights_5=0.0363
    ConstraintWeights_6=0.0264
    ConstraintWeights_7=0.0135
   ) {
    alphaBoolBlend {blendOption=Linear}
    boneToModify {boneName=dyn_belt_l_MSTR}
    constraintSetup [
     {targetBone {boneName=spine_01}}
     {targetBone {boneName=pelvis}}
     {targetBone {boneName=spine_02}}
     {targetBone {boneName=spine_03}}
     {targetBone {boneName=dyn_coat_bk_l_01}}
     {targetBone {boneName=spine_04}}
     {targetBone {boneName=dyn_coat_bk_l_02}}
     {targetBone {boneName=dyn_coat_bk_01}}
    ]
    constraintWeights [1,1,1,1,1,1,1,1]
   }
   n67 = RigidBody(
    ComponentPose=@n66
    OverrideWorldGravity=@n63
    SimSpaceSettings=(WorldAlpha=1,VelocityScaleZ=0,DampingAlpha=1,MaxLinearVelocity=500,MaxAngularVelocity=3,MaxLinearAcceleration=700,MaxAngularAcceleration=1000,ExternalLinearDragV=(X=0.1,Y=0.1,Z=0),ExternalLinearVelocity=(X=0,Y=0,Z=0),ExternalAngularVelocity=(X=0,Y=0,Z=0))
   ) {
    bClampLinearTranslationLimitToRefPose=true
    bOverrideWorldGravity=true
    overridePhysicsAsset=&/M_MED_PuffinSmile_Dyn_belt
    simulationSpace=BaseBoneSpace
    alphaBoolBlend {blendOption=Linear}
    baseBoneRef {boneName=spine_01}
    overrideWorldGravity {z=-980}
    simSpaceSettings {
     maxAngularAcceleration=1000
     maxAngularVelocity=3
     maxLinearAcceleration=700
     maxLinearVelocity=500
     velocityScaleZ=0
     worldAlpha=1
     externalLinearDragV=<0.1 0.1>
    }
   }
   n68 = RigidBody(
    ComponentPose=Local To Component(LocalPose=Use cached pose 'StartupCache'())
    OverrideWorldGravity=@n63
    SimSpaceSettings=(WorldAlpha=1,VelocityScaleZ=0,DampingAlpha=1,MaxLinearVelocity=500,MaxAngularVelocity=3,MaxLinearAcceleration=700,MaxAngularAcceleration=1000,ExternalLinearDragV=(X=0.3,Y=0.3,Z=0),ExternalLinearVelocity=(X=0,Y=0,Z=0),ExternalAngularVelocity=(X=0,Y=0,Z=0))
   ) {
    bClampLinearTranslationLimitToRefPose=true
    bOverrideWorldGravity=true
    overridePhysicsAsset=&/M_MED_PuffinSmile_Dyn_coat_for_cloth
    simulationSpace=BaseBoneSpace
    alphaBoolBlend {blendOption=Linear}
    baseBoneRef {boneName=spine_01}
    overrideWorldGravity {z=-980}
    simSpaceSettings {
     maxAngularAcceleration=1000
     maxAngularVelocity=3
     maxLinearAcceleration=700
     maxLinearVelocity=500
     velocityScaleZ=0
     worldAlpha=1
     externalLinearDragV=<0.3 0.3>
    }
   }
   n69 = "Constraint - dyn_belt_r_MSTR (PPPPPPPP)"(
    ComponentPose=@n68
    Alpha=1
    ConstraintWeights_0=0.365
    ConstraintWeights_1=0.3055
    ConstraintWeights_2=0.301
    ConstraintWeights_3=0.0143
    ConstraintWeights_4=0.0115
    ConstraintWeights_5=0.001
    ConstraintWeights_6=0.0009
    ConstraintWeights_7=0.0008
   ) {
    alphaBoolBlend {blendOption=Linear}
    boneToModify {boneName=dyn_belt_r_MSTR}
    constraintSetup [
     {targetBone {boneName=spine_02}}
     {targetBone {boneName=spine_01}}
     {targetBone {boneName=pelvis}}
     {targetBone {boneName=dyn_coat_bk_r_01}}
     {targetBone {boneName=dyn_coat_r_03}}
     {targetBone {boneName=dyn_coat_bk_r_02}}
     {targetBone {boneName=spine_03}}
     {targetBone {boneName=dyn_coat_r_02}}
    ]
    constraintWeights [1,1,1,1,1,1,1,1]
   }
   n70 = "Constraint - dyn_belt_l_MSTR (PPPPPPPP)"(
    ComponentPose=@n69
    Alpha=1
    ConstraintWeights_0=0.2744
    ConstraintWeights_1=0.2648
    ConstraintWeights_2=0.2016
    ConstraintWeights_3=0.1424
    ConstraintWeights_4=0.0407
    ConstraintWeights_5=0.0363
    ConstraintWeights_6=0.0264
    ConstraintWeights_7=0.0135
   ) {
    alphaBoolBlend {blendOption=Linear}
    boneToModify {boneName=dyn_belt_l_MSTR}
    constraintSetup [
     {targetBone {boneName=spine_01}}
     {targetBone {boneName=pelvis}}
     {targetBone {boneName=spine_02}}
     {targetBone {boneName=spine_03}}
     {targetBone {boneName=dyn_coat_bk_l_01}}
     {targetBone {boneName=spine_04}}
     {targetBone {boneName=dyn_coat_bk_l_02}}
     {targetBone {boneName=dyn_coat_bk_01}}
    ]
    constraintWeights [1,1,1,1,1,1,1,1]
   }
   n71 = RigidBody(
    ComponentPose=@n70
    OverrideWorldGravity=@n63
    SimSpaceSettings=(WorldAlpha=1,VelocityScaleZ=0,DampingAlpha=1,MaxLinearVelocity=500,MaxAngularVelocity=3,MaxLinearAcceleration=700,MaxAngularAcceleration=1000,ExternalLinearDragV=(X=0.1,Y=0.1,Z=0),ExternalLinearVelocity=(X=0,Y=0,Z=0),ExternalAngularVelocity=(X=0,Y=0,Z=0))
   ) {
    bClampLinearTranslationLimitToRefPose=true
    bOverrideWorldGravity=true
    overridePhysicsAsset=&/M_MED_PuffinSmile_Dyn_belt_for_cloth
    simulationSpace=BaseBoneSpace
    alphaBoolBlend {blendOption=Linear}
    baseBoneRef {boneName=spine_01}
    overrideWorldGravity {z=-980}
    simSpaceSettings {
     maxAngularAcceleration=1000
     maxAngularVelocity=3
     maxLinearAcceleration=700
     maxLinearVelocity=500
     velocityScaleZ=0
     worldAlpha=1
     externalLinearDragV=<0.1 0.1>
    }
   }
   Save cached pose 'RbanCache'(Pose=Blend Poses by bool(
     BlendPose_0=Component To Local(ComponentPose=@n71)
     BlendPose_1=Component To Local(ComponentPose=@n67)
     BlendTime_0=0.2
     BlendTime_1=0.5
     bActiveValue=Get bClothEnabled()
    ) {blendType=Linear}) {cacheName=RbanCache}
   Get TrailControllerRot() -> (TrailControllerRot:Rotator)
  ]}
 EventGraph {nodes [
   n1 = Event BlueprintUpdateAnimation()
   > Set DeltaTimeLocal(DeltaTimeLocal:double=@n1.DeltaTimeX) -> (Output_Get:double)
   > n2 = Sequence!(execute) -> (then_0, then_1, then_2:Exec, then_3:Exec)
   > n3 = Branch(Condition=Get bClothEnabled())
   > n4 = Sequence!(execute) -> (then_0, then_1)
   n5 = Gravity Override State Function DEPRECATED(
    execute=[@n4.then_0, @n3.else]
    CharacterPartAnimInstance=Self-Reference()
    GameTimeInSeconds=GetGameTimeInSeconds()
    GravityOverrideParams=Get Gravity Override Params()
    RBAN Params=Get RBAN Params()
   )
   > n6 = Set GravityOverride(GravityOverride=@n5)
   > n7 = Set RBAN Settings Out(RBAN Settings Out:RBANSetting Struct Structure=@n5.RBAN SettingsOut) -> (Output_Get:RBANSetting Struct Structure)
   > Set RBAN Settings Belt for Cloth(RBAN Settings Belt for Cloth:RBANSetting Struct Structure=@n7) -> (Output_Get:RBANSetting Struct Structure)
   n8 = Cloth Controller Function DEPRECATED!(
    execute=@n4.then_1
    CharacterPartAnimInstance:Custom Character Part Anim Instance Object Reference=Self-Reference()
    ClothingAssetName:String=M_MED_PuffinSmile_Clothing_0
    Cloth Paramaters:Cloth Params Struct Structure=Get Cloth Paramaters_0() -> (Cloth Paramaters_0:Cloth Params Struct Structure)
    Cloth Core Settings:Cloth Core Setting Struct Structure=Get Cloth Core Settings_0() -> (Cloth Core Settings_0:Cloth Core Setting Struct Structure)
    ClothUpAxis:Vector="1,0,0"
    OnGround_Moving_SpeedLimit:double=2
    OnGround_Moving_AngularVel_Max:double=0.05
    OnGround_Moving_FictitiousVel_Max:double=0.6
    LinVelScale_FwdBwd_Clamp:double=0
    LinVelScale_UpDwn_Clamp:double=0.7
    DeltaTime:double=Get DeltaTimeLocal() -> (DeltaTimeLocal:double)
    GameTimeInSeconds:double=GetGameTimeInSeconds()
    WindGustsParams:Wind Gust Params Struct Structure=Get Wind Gust Params() -> (Wind Gust Params:Wind Gust Params Struct Structure)
    WindGustLocalSettings:Wind Gust Local Setting Struct Structure=Get Gust Local Settings() -> (Gust Local Settings:Wind Gust Local Setting Struct Structure)
   ) -> (
    then
    CurrentStateOut:String
    Cloth Settings Out:Cloth Setting Struct Structure
    Cloth Core Settings Out:Cloth Core Setting Struct Structure
    WindGustSettings Out:Wind Gust Setting Struct Structure
    WindGustLocalSettings Out:Wind Gust Local Setting Struct Structure
    GustStrength - WindVelocityOut:Vector
    LinearVelVector-DELETE LATER:Vector
   )
   > Set Gust Local Settings(Gust Local Settings:Wind Gust Local Setting Struct Structure=@n8.WindGustLocalSettings Out) -> (Output_Get:Wind Gust Local Setting Struct Structure)
   n9 = PlayerMovement(execute=@n2.then_1, AnimInstance=Self-Reference(), XDotProductCoefficient=1, YDotProductCoefficient=1, ZDotProductCoefficient=1)
   n10 = MapRangeClamped(Value=@n9.YMovement, InRangeA=0, InRangeB=1)
   n11 = "Absolute (Float)"(A=@n9)
   > Set BwdMovement(BwdMovement:double=MapRangeClamped(Value=float + float(A=@n11, B=@n10) {is=title K2Node_CommutativeAssociativeBinaryOperator}, InRangeA=0, InRangeB=1.4)) -> (Output_Get:double)
   n12 = Branch(execute=@n2.then_2, Condition=Get bIsSwimSprinting() -> (bIsSwimSprinting:bool))
   > Set TrailControllerRot(TrailControllerRot:Rotator="15,0, 0") -> (Output_Get:Rotator)
   n13 = Branch(execute=@n12.else, Condition=Get bIsSkyDiving())
   > n14 = Branch(Condition=Get bIsParachuteOpened())
   > Set TrailControllerRot(TrailControllerRot="45,0, 0")
   n15 = Branch(execute=@n14.else, Condition="float > float"(A=Get PawnVelocityZ(), B=50) {is=title K2Node_CallFunction})
   > Set BwdMovement(BwdMovement=float * float(A=Get BwdMovement(), B=0))
   > Set TrailControllerRot(TrailControllerRot="45,0, 0")
   n16 = Branch(execute=@n15.else, Condition=Get bIsSkydiveDiveMode())
   > Set TrailControllerRot(TrailControllerRot="15,0, 0")
   Set TrailControllerRot(execute=@n16.else, TrailControllerRot="60,0, 0")
   Set TrailControllerRot(execute=@n13.else, TrailControllerRot="0, 0, 0")
   > Set BwdMovement(BwdMovement=float * float(A=Get BwdMovement(), B=1))
   Event OnBeginSkydiving()
   > Set TrailControllerRot(TrailControllerRot="-65,0,0")
   Event OnEndSkydiving()
   > Set TrailControllerRot(TrailControllerRot="-10,0,0")
   Break Sim Space Settings!(SimSpaceSettings:"Sim Space Settings Structure (by ref)"=Break RBANSetting Struct!(RBANSettingStruct:"RBANSetting Struct Structure (by ref)"=@n7) -> (
     ComponentLinearAccScale_28_#1:Vector=~
     ComponentLinearVelScale_31_#2:Vector=~
     SimSpaceSettings_34_#3:Sim Space Settings Structure
    ) {is=title K2Node_BreakStruct}.SimSpaceSettings_34_#3) -> (
    WorldAlpha:float=~
    VelocityScaleZ:float=~
    DampingAlpha:float=~
    MaxLinearVelocity:float=~
    MaxAngularVelocity:float=~
    MaxLinearAcceleration:float=~
    MaxAngularAcceleration:float=~
    ExternalLinearDragV:Vector=~
    ExternalLinearVelocity:Vector=~
    ExternalAngularVelocity:Vector=~
   ) {is=title K2Node_BreakStruct}
   n17 = GetPartSkeletalMeshComponent()
   DrawVectorFromJointFunction!(
    execute
    SkelMeshComponent:Scene Component Object Reference=@n17
    JointName:Name=pelvis
    LineEnd:Vector=@n6
    JointOffset:Vector="0, 0, 0"
    LineColor:Linear Color Structure=(R=1,G=0.9595,B=0,A=1)
    Thickness:double=2.2
    Foreground:EDrawDebugSceneDepthPriorityGroup Enum=Foreground
   ) -> (then)
   DrawVectorFromJointFunction(
    SkelMeshComponent=@n17
    JointName=pelvis
    LineEnd=Break Cloth Setting Struct!(ClothSettingStruct:"Cloth Setting Struct Structure (by ref)"=@n8.Cloth Settings Out) -> (
     LinearVelocityScale_7_#1:Vector=~
     AngularVelocityScale_6_#2:float=~
     FictitiousAngularScale_10_#4:float=~
     GravityOverride_15_#5:Vector=~
     DragCoefficient_23_#3:Vector 2D Structure
     LiftCoefficient_24_#6:Vector 2D Structure
     JointName_27_#7:Name
    ) {is=title K2Node_BreakStruct}.GravityOverride_15_#5
    JointOffset="0, 0, 0"
    LineColor=(R=1,G=0,B=0,A=1)
    Thickness=2
    Foreground=Foreground
   )
  ]}
]
variables [
 BwdMovement {type=number}
 Cloth Core Settings_0 {
  type=ClothCoreSettingStruct
  default {
   collisionThickness=0
   dampingCoefficient=0.2
   frictionCoefficient=0.3
   localDampingCoefficient=0
   numberIterations=1
   numberSubsteps=2
   selfCollisionThickness=0
   animDriveDamping=<0 1>
   animDriveStiffness=<0 1>
   areaStiffness=<1 1>
   bendingStiffness=<1 1>
   edgeStiffness=<1 1>
   tetherStiffness=<1 1>
  }
 }
 Cloth Paramaters_0 {
  type=ClothParamsStruct
  default {
   dBNO {
    angularVelocityScale=0.6
    fictitiousAngularScale=0.1
    jointName=pelvis
    dragCoefficient=<0.2 1>
    gravityOverride=<-980 400 0>
    liftCoefficient=<0 0.5>
    linearVelocityScale=<0.3 0.3 0.1>
   }
   dBNO_Carried {
    angularVelocityScale=0.6
    fictitiousAngularScale=0.1
    jointName=pelvis
    dragCoefficient=<0.2 1>
    gravityOverride=<-980 -200 0>
    liftCoefficient=<0 0.5>
    linearVelocityScale=<0.3 0.3 0.1>
   }
   dBNO_Impulse {
    angularVelocityScale=0
    fictitiousAngularScale=0
    jointName=pelvis
    dragCoefficient=<0 0.2>
    gravityOverride=<-3000 -2000 0>
    liftCoefficient=<0 0.1>
    linearVelocityScale=<0.3 0.1 0.3>
   }
   dBNO_RollBwd {
    angularVelocityScale=0
    fictitiousAngularScale=0
    jointName=pelvis
    dragCoefficient=<0 0.2>
    gravityOverride=<-2000 0 0>
    liftCoefficient=<0 0.1>
    linearVelocityScale=<0.3 0.1 0.3>
   }
   dBNO_RollFwd {
    angularVelocityScale=0
    fictitiousAngularScale=0
    jointName=pelvis
    dragCoefficient=<0 0.2>
    gravityOverride=<-980 0 0>
    liftCoefficient=<0 0.1>
    linearVelocityScale=<0.3 0.1 0.3>
   }
   dBNO_Tumbling {
    angularVelocityScale=0
    fictitiousAngularScale=0
    jointName=pelvis
    dragCoefficient=<0 0.2>
    gravityOverride=<-2500 -300 0>
    liftCoefficient=<0 0.1>
    linearVelocityScale=<0.3 0.1 0.3>
   }
   emoteOrMelee {
    angularVelocityScale=0.2
    fictitiousAngularScale=0.1
    jointName=pelvis
    dragCoefficient=<0 0.3>
    gravityOverride=<-980 0 0>
    liftCoefficient=<0 0.2>
    linearVelocityScale=<0.05 0.05 0.05>
   }
   front_End {
    angularVelocityScale=0
    fictitiousAngularScale=0
    jointName=root
    dragCoefficient=<0 0>
    gravityOverride=<0 0 0>
    liftCoefficient=<0 0>
    linearVelocityScale=<0 0 0>
   }
   onGround_Crouch_Moving {
    angularVelocityScale=0.1
    fictitiousAngularScale=0.3
    jointName=root
    dragCoefficient=<2 2>
    gravityOverride=<0 0 -980>
    liftCoefficient=<0.6 0.6>
    linearVelocityScale=<0.6 0.6 0.5>
   }
   onGround_Falling {
    angularVelocityScale=0.6
    fictitiousAngularScale=0
    jointName=root
    dragCoefficient=<0 0.2>
    gravityOverride=<0 -500 -980>
    liftCoefficient=<0 0.1>
    linearVelocityScale=<2 1 0.3>
   }
   onGround_Moving {
    angularVelocityScale=0.4
    fictitiousAngularScale=0.1
    jointName=root
    dragCoefficient=<0 0.4>
    gravityOverride=<0 -100 -980>
    liftCoefficient=<0 0.2>
    linearVelocityScale=<0.25 0.3 0.3>
   }
   onGround_Standing {
    angularVelocityScale=0.05
    fictitiousAngularScale=0.1
    jointName=root
    dragCoefficient=<0 0.2>
    gravityOverride=<0 -100 -980>
    liftCoefficient=<0 0.2>
    linearVelocityScale=<0.4 0.2 0.2>
   }
   rail_Grinding {
    angularVelocityScale=0.2
    fictitiousAngularScale=0
    jointName=root
    dragCoefficient=<0.1 2>
    gravityOverride=<300 -500 -800>
    liftCoefficient=<0.1 2>
    linearVelocityScale=<0 0 0>
   }
   riding_Moving {
    angularVelocityScale=0.05
    fictitiousAngularScale=1
    jointName=root
    dragCoefficient=<0.5 0.5>
    gravityOverride=<0 0 -980>
    liftCoefficient=<0.5 0.5>
    linearVelocityScale=<0.4 0.6 0.3>
   }
   riding_Moving_Targeting {
    angularVelocityScale=0.05
    fictitiousAngularScale=0.7
    jointName=root
    dragCoefficient=<0.5 0.5>
    gravityOverride=<0 0 -980>
    liftCoefficient=<0.5 0.5>
    linearVelocityScale=<0.2 0.6 0.2>
   }
   riding_Standing {
    angularVelocityScale=0.05
    fictitiousAngularScale=1
    jointName=root
    dragCoefficient=<0.5 0.5>
    gravityOverride=<0 -450 -980>
    liftCoefficient=<0.5 0.5>
    linearVelocityScale=<0.6 0.6 0.3>
   }
   skydive_DivingDown {
    angularVelocityScale=0.1
    fictitiousAngularScale=0
    jointName=pelvis
    dragCoefficient=<0 1>
    gravityOverride=<-450 -100 0>
    liftCoefficient=<0 1>
    linearVelocityScale=<0.1 0.1 0.1>
   }
   skydive_FloatingDown {
    angularVelocityScale=0
    fictitiousAngularScale=0
    jointName=pelvis
    dragCoefficient=<0 0>
    gravityOverride=<0 0 0>
    liftCoefficient=<0 0>
    linearVelocityScale=<0 0 0>
   }
   skydive_GoingUp {
    angularVelocityScale=0.1
    fictitiousAngularScale=0
    jointName=root
    dragCoefficient=<0 0>
    gravityOverride=<0 -3000 3000>
    liftCoefficient=<0 0>
    linearVelocityScale=<0.1 0.1 0.05>
   }
   skydive_Parachuting {
    angularVelocityScale=0.1
    fictitiousAngularScale=0
    jointName=root
    dragCoefficient=<0 1>
    gravityOverride=<0 -500 -980>
    liftCoefficient=<0 1>
    linearVelocityScale=<0.05 0.05 0.1>
   }
   skydive_Surfing {
    angularVelocityScale=0.6
    fictitiousAngularScale=0.1
    jointName=pelvis
    dragCoefficient=<1.5 1.5>
    gravityOverride=<0 -600 -500>
    liftCoefficient=<1 1>
    linearVelocityScale=<0.1 0.1 0.3>
   }
   swimming {
    angularVelocityScale=0.6
    fictitiousAngularScale=0
    jointName=root
    dragCoefficient=<0 2>
    gravityOverride=<0 -600 400>
    liftCoefficient=<0 0.4>
    linearVelocityScale=<0.6 0.2 0.2>
   }
   tactical_Sprint {
    angularVelocityScale=0.3
    fictitiousAngularScale=0.1
    jointName=root
    dragCoefficient=<0 1>
    gravityOverride=<0 0 -980>
    liftCoefficient=<0 0.75>
    linearVelocityScale=<0.25 0.3 0.3>
   }
   ziplining {
    angularVelocityScale=0.1
    fictitiousAngularScale=0
    jointName=root
    dragCoefficient=<0 0.5>
    gravityOverride=<0 0 -980>
    liftCoefficient=<0 0.3>
    linearVelocityScale=<0.1 0.1 0.1>
   }
  }
 }
 DeltaTimeLocal {type=number}
 Gravity Override Params {
  type=GravityOverrideParamsStruct
  default {
   rail_Grinding {gravityOverride=<500 -500 -250>, windAmplitude=<80 95 62>, windFrequency=<5 6 8>, windOffset=<50 -500 -400>}
   skydive_GoingUp {gravityOverride=<0 -600 2500>}
  }
 }
 GravityOverride {type=Vector}
 Gust Local Settings {type=WindGustLocalSettingStruct, default {elapsedTime=0, finalGustStrengthLocal=0, gustDuration=0, gustStrength=0, timeBetweenGusts=0}}
 RBAN Params {
  type=RBANParamsStruct
  default {
   dBNO {simSpaceSettings {
     maxAngularAcceleration=3
     maxAngularVelocity=5
     maxLinearAcceleration=500
     maxLinearVelocity=1500
     velocityScaleZ=1
     externalLinearDragV=<1 1 0>
    }}
   dBNO_Carried {simSpaceSettings {
     maxAngularAcceleration=3
     maxAngularVelocity=5
     maxLinearAcceleration=500
     maxLinearVelocity=1500
     velocityScaleZ=1
     externalLinearDragV=<1 1 0>
    }}
   emoteOrMelee {simSpaceSettings {
     maxAngularAcceleration=3
     maxAngularVelocity=5
     maxLinearAcceleration=500
     maxLinearVelocity=1500
     velocityScaleZ=1
     externalLinearDragV=<1 1 0>
    }}
   front_End {simSpaceSettings {
     maxAngularAcceleration=3
     maxAngularVelocity=5
     maxLinearAcceleration=500
     maxLinearVelocity=1500
     velocityScaleZ=1
     externalLinearDragV=<1 1 0>
    }}
   onGround_Crouch_Moving {simSpaceSettings {
     maxAngularAcceleration=3
     maxAngularVelocity=5
     maxLinearAcceleration=500
     maxLinearVelocity=1500
     velocityScaleZ=1
     externalLinearDragV=<1 1 0>
    }}
   onGround_Falling {simSpaceSettings {
     maxAngularAcceleration=3
     maxAngularVelocity=5
     maxLinearAcceleration=500
     maxLinearVelocity=1500
     velocityScaleZ=1
     externalLinearDragV=<1 1 0>
    }}
   onGround_Moving {
    componentLinearAccScale=<0 0 0>
    componentLinearVelScale=<0 0 0>
    simSpaceSettings {
     maxAngularAcceleration=5
     maxAngularVelocity=10
     maxLinearAcceleration=500
     maxLinearVelocity=1500
     velocityScaleZ=1
     externalLinearDragV=<1 1 0>
    }
   }
   onGround_Standing {
    componentLinearAccScale=<0 0 0>
    componentLinearVelScale=<0 0 0>
    simSpaceSettings {
     maxAngularAcceleration=3
     maxAngularVelocity=5
     maxLinearAcceleration=500
     maxLinearVelocity=1500
     velocityScaleZ=1
     externalLinearDragV=<1 1 0>
    }
   }
   rail_Grinding {simSpaceSettings {
     maxAngularAcceleration=3
     maxAngularVelocity=5
     maxLinearAcceleration=500
     maxLinearVelocity=1500
     velocityScaleZ=1
     worldAlpha=1
     externalLinearDragV=<1 1 0>
    }}
   riding_Moving {simSpaceSettings {
     maxAngularAcceleration=3
     maxAngularVelocity=5
     maxLinearAcceleration=500
     maxLinearVelocity=1500
     velocityScaleZ=1
     externalLinearDragV=<1 1 0>
    }}
   riding_Moving_Targeting {simSpaceSettings {
     maxAngularAcceleration=3
     maxAngularVelocity=5
     maxLinearAcceleration=500
     maxLinearVelocity=1500
     velocityScaleZ=1
     externalLinearDragV=<1 1 0>
    }}
   riding_Standing {simSpaceSettings {
     maxAngularAcceleration=3
     maxAngularVelocity=5
     maxLinearAcceleration=500
     maxLinearVelocity=1500
     velocityScaleZ=1
     externalLinearDragV=<1 1 0>
    }}
   skydive_DivingDown {
    componentLinearAccScale=<0 0 0>
    componentLinearVelScale=<0 0 0>
    simSpaceSettings {
     maxAngularAcceleration=3
     maxAngularVelocity=5
     maxLinearAcceleration=500
     maxLinearVelocity=1500
     velocityScaleZ=1
     externalLinearDragV=<1 1 0>
    }
   }
   skydive_FloatingDown {
    componentLinearAccScale=<0 0 0>
    componentLinearVelScale=<0 0 0>
    simSpaceSettings {
     maxAngularAcceleration=3
     maxAngularVelocity=5
     maxLinearAcceleration=500
     maxLinearVelocity=1500
     velocityScaleZ=1
     externalLinearDragV=<1 1 0>
    }
   }
   skydive_GoingUp {simSpaceSettings {
     maxAngularAcceleration=3
     maxAngularVelocity=5
     maxLinearAcceleration=500
     maxLinearVelocity=1500
     velocityScaleZ=1
     externalLinearDragV=<1 1 0>
    }}
   skydive_Parachuting {
    componentLinearAccScale=<0 0 0>
    componentLinearVelScale=<0 0 0>
    simSpaceSettings {
     maxAngularAcceleration=3
     maxAngularVelocity=5
     maxLinearAcceleration=500
     maxLinearVelocity=1500
     velocityScaleZ=1
     externalLinearDragV=<1 1 0>
    }
   }
   skydive_Surfing {simSpaceSettings {
     maxAngularAcceleration=3
     maxAngularVelocity=5
     maxLinearAcceleration=500
     maxLinearVelocity=1500
     velocityScaleZ=1
     externalLinearDragV=<1 1 0>
    }}
   swimming {simSpaceSettings {
     maxAngularAcceleration=3
     maxAngularVelocity=5
     maxLinearAcceleration=500
     maxLinearVelocity=1500
     velocityScaleZ=1
     externalLinearDragV=<1 1 0>
    }}
   tactical_Sprint {
    componentLinearAccScale=<0 0 0>
    componentLinearVelScale=<0 0 0>
    simSpaceSettings {
     maxAngularAcceleration=3
     maxAngularVelocity=5
     maxLinearAcceleration=500
     maxLinearVelocity=1500
     velocityScaleZ=1
     externalLinearDragV=<1 1 0>
    }
   }
   ziplining {simSpaceSettings {
     maxAngularAcceleration=0.5
     maxAngularVelocity=1
     maxLinearAcceleration=50
     maxLinearVelocity=200
     velocityScaleZ=1
     externalLinearDragV=<0.5 0.5 0>
    }}
  }
 }
 RBAN Settings Belt for Cloth {
  type=RBANSettingStruct
  default {
   componentLinearAccScale=<0 0 0>
   componentLinearVelScale=<0 0 0>
   simSpaceSettings {
    dampingAlpha=1
    maxAngularAcceleration=1000
    maxAngularVelocity=3
    maxLinearAcceleration=300
    maxLinearVelocity=700
    velocityScaleZ=0
    worldAlpha=1
    externalAngularVelocity=<0 0 0>
    externalLinearDragV=<0 0.5 0>
    externalLinearVelocity=<0 0 0>
   }
  }
 }
 RBAN Settings Out {
  type=RBANSettingStruct
  default {
   componentLinearAccScale=<0 0 0>
   componentLinearVelScale=<0 0 0>
   simSpaceSettings {
    dampingAlpha=1
    maxAngularAcceleration=1000
    maxAngularVelocity=3
    maxLinearAcceleration=300
    maxLinearVelocity=700
    velocityScaleZ=0
    worldAlpha=1
    externalAngularVelocity=<0 0 0>
    externalLinearDragV=<0 0.5 0>
    externalLinearVelocity=<0 0 0>
   }
  }
 }
 TrailControllerRot {type=Rotator}
 Wind Gust Params {
  type=WindGustParamsStruct
  default {
   dBNO {
    maxDownInterpSpeed=5
    maxGustDuration=2
    maxGustStrength=7
    maxTimeBetweenGusts=7
    maxUpInterpSpeed=5
    minDownInterpSpeed=2
    minGustDuration=0.5
    minGustStrength=3
    minTimeBetweenGusts=0
    minUpInterpSpeed=2
    windTurbulenceScalarA=0.1
    windTurbulenceScalarB=1
    windAmplitude=<0 0 0>
    windFrequency=<0 0 0>
    windOffset=<0 0 0>
   }
   dBNO_Carried {
    maxDownInterpSpeed=5
    maxGustDuration=2
    maxGustStrength=7
    maxTimeBetweenGusts=7
    maxUpInterpSpeed=5
    minDownInterpSpeed=2
    minGustDuration=0.5
    minGustStrength=3
    minTimeBetweenGusts=0
    minUpInterpSpeed=2
    windTurbulenceScalarA=0.1
    windTurbulenceScalarB=1
    windAmplitude=<0 0 0>
    windFrequency=<0 0 0>
    windOffset=<0 0 0>
   }
   dBNO_Impulse {
    maxDownInterpSpeed=5
    maxGustDuration=2
    maxGustStrength=7
    maxTimeBetweenGusts=7
    maxUpInterpSpeed=5
    minDownInterpSpeed=2
    minGustDuration=0.5
    minGustStrength=3
    minTimeBetweenGusts=0
    minUpInterpSpeed=2
    windTurbulenceScalarA=0.1
    windTurbulenceScalarB=1
    windAmplitude=<0 0 0>
    windFrequency=<0 0 0>
    windOffset=<0 0 0>
   }
   dBNO_RollBwd {
    maxDownInterpSpeed=5
    maxGustDuration=2
    maxGustStrength=7
    maxTimeBetweenGusts=7
    maxUpInterpSpeed=5
    minDownInterpSpeed=2
    minGustDuration=0.5
    minGustStrength=3
    minTimeBetweenGusts=0
    minUpInterpSpeed=2
    windTurbulenceScalarA=0.1
    windTurbulenceScalarB=1
    windAmplitude=<0 0 0>
    windFrequency=<0 0 0>
    windOffset=<0 0 0>
   }
   dBNO_RollFwd {
    maxDownInterpSpeed=5
    maxGustDuration=2
    maxGustStrength=7
    maxTimeBetweenGusts=7
    maxUpInterpSpeed=5
    minDownInterpSpeed=2
    minGustDuration=0.5
    minGustStrength=3
    minTimeBetweenGusts=0
    minUpInterpSpeed=2
    windTurbulenceScalarA=0.1
    windTurbulenceScalarB=1
    windAmplitude=<0 0 0>
    windFrequency=<0 0 0>
    windOffset=<0 0 0>
   }
   dBNO_Tumbling {
    maxDownInterpSpeed=5
    maxGustDuration=2
    maxGustStrength=7
    maxTimeBetweenGusts=7
    maxUpInterpSpeed=5
    minDownInterpSpeed=2
    minGustDuration=0.5
    minGustStrength=3
    minTimeBetweenGusts=0
    minUpInterpSpeed=2
    windTurbulenceScalarA=0.1
    windTurbulenceScalarB=1
    windAmplitude=<0 0 0>
    windFrequency=<0 0 0>
    windOffset=<0 0 0>
   }
   emoteOrMelee {
    maxDownInterpSpeed=5
    maxGustDuration=2
    maxGustStrength=7
    maxTimeBetweenGusts=7
    maxUpInterpSpeed=5
    minDownInterpSpeed=2
    minGustDuration=0.5
    minGustStrength=3
    minTimeBetweenGusts=0
    minUpInterpSpeed=2
    windTurbulenceScalarA=0.1
    windTurbulenceScalarB=1
    windAmplitude=<30 25 50>
    windFrequency=<2 3 5>
    windOffset=<-300 -150 0>
   }
   front_End {
    maxDownInterpSpeed=5
    maxGustDuration=2
    maxGustStrength=4
    maxTimeBetweenGusts=7
    maxUpInterpSpeed=5
    minDownInterpSpeed=2
    minGustDuration=0.5
    minGustStrength=2
    minTimeBetweenGusts=0
    minUpInterpSpeed=2
    windTurbulenceScalarA=0.1
    windTurbulenceScalarB=1
    windAmplitude=<10 30 50>
    windFrequency=<10 3 4>
    windOffset=<0 -200 -250>
   }
   onGround_Crouch_Moving {
    maxDownInterpSpeed=5
    maxGustDuration=2
    maxGustStrength=7
    maxTimeBetweenGusts=7
    maxUpInterpSpeed=5
    minDownInterpSpeed=2
    minGustDuration=0.5
    minGustStrength=3
    minTimeBetweenGusts=0
    minUpInterpSpeed=2
    windTurbulenceScalarA=0.1
    windTurbulenceScalarB=1
    windAmplitude=<100 25 75>
    windFrequency=<1.5 0.5 7.5>
    windOffset=<0 -100 -175>
   }
   onGround_Falling {
    maxDownInterpSpeed=5
    maxGustDuration=2
    maxGustStrength=7
    maxTimeBetweenGusts=7
    maxUpInterpSpeed=5
    minDownInterpSpeed=2
    minGustDuration=0.5
    minGustStrength=3
    minTimeBetweenGusts=0
    minUpInterpSpeed=2
    windTurbulenceScalarA=0.1
    windTurbulenceScalarB=1
    windAmplitude=<0 0 0>
    windFrequency=<0 0 0>
    windOffset=<0 0 0>
   }
   onGround_Moving {
    maxDownInterpSpeed=5
    maxGustDuration=8
    maxGustStrength=7
    maxTimeBetweenGusts=8
    maxUpInterpSpeed=5
    minDownInterpSpeed=2
    minGustDuration=0.2
    minGustStrength=3
    minTimeBetweenGusts=1
    minUpInterpSpeed=2
    windTurbulenceScalarA=0.1
    windTurbulenceScalarB=0.8
    windAmplitude=<30 65 42>
    windFrequency=<5 6 8.6>
    windOffset=<30 -250 -300>
   }
   onGround_Standing {
    maxDownInterpSpeed=5
    maxGustDuration=8
    maxGustStrength=7
    maxTimeBetweenGusts=8
    maxUpInterpSpeed=5
    minDownInterpSpeed=2
    minGustDuration=0.2
    minGustStrength=2
    minTimeBetweenGusts=1
    minUpInterpSpeed=2
    windTurbulenceScalarA=0.1
    windTurbulenceScalarB=0.8
    windAmplitude=<75 24 28>
    windFrequency=<1.5 2.3 3.7>
    windOffset=<75 -100 -150>
   }
   rail_Grinding {
    maxDownInterpSpeed=5
    maxGustDuration=8
    maxGustStrength=7
    maxTimeBetweenGusts=8
    maxUpInterpSpeed=5
    minDownInterpSpeed=2
    minGustDuration=0.2
    minGustStrength=3
    minTimeBetweenGusts=1
    minUpInterpSpeed=2
    windTurbulenceScalarA=0.1
    windTurbulenceScalarB=0.8
    windAmplitude=<20 65 42>
    windFrequency=<5 6 8.6>
    windOffset=<50 -500 -400>
   }
   riding_Moving {
    maxDownInterpSpeed=5
    maxGustDuration=2
    maxGustStrength=7
    maxTimeBetweenGusts=7
    maxUpInterpSpeed=5
    minDownInterpSpeed=2
    minGustDuration=0.5
    minGustStrength=3
    minTimeBetweenGusts=0
    minUpInterpSpeed=2
    windTurbulenceScalarA=0.1
    windTurbulenceScalarB=1
    windAmplitude=<0 0 0>
    windFrequency=<0 0 0>
    windOffset=<0 0 0>
   }
   riding_Moving_Targeting {
    maxDownInterpSpeed=5
    maxGustDuration=2
    maxGustStrength=7
    maxTimeBetweenGusts=7
    maxUpInterpSpeed=5
    minDownInterpSpeed=2
    minGustDuration=0.5
    minGustStrength=3
    minTimeBetweenGusts=0
    minUpInterpSpeed=2
    windTurbulenceScalarA=0.1
    windTurbulenceScalarB=1
    windAmplitude=<0 0 0>
    windFrequency=<0 0 0>
    windOffset=<0 0 0>
   }
   riding_Standing {
    maxDownInterpSpeed=5
    maxGustDuration=2
    maxGustStrength=7
    maxTimeBetweenGusts=7
    maxUpInterpSpeed=5
    minDownInterpSpeed=2
    minGustDuration=0.5
    minGustStrength=3
    minTimeBetweenGusts=0
    minUpInterpSpeed=2
    windTurbulenceScalarA=0.1
    windTurbulenceScalarB=1
    windAmplitude=<0 0 0>
    windFrequency=<0 0 0>
    windOffset=<0 0 0>
   }
   skydive_DivingDown {
    maxDownInterpSpeed=5
    maxGustDuration=8
    maxGustStrength=7
    maxTimeBetweenGusts=8
    maxUpInterpSpeed=5
    minDownInterpSpeed=2
    minGustDuration=0.2
    minGustStrength=3
    minTimeBetweenGusts=1
    minUpInterpSpeed=2
    windTurbulenceScalarA=0.1
    windTurbulenceScalarB=4
    windAmplitude=<0 25 35>
    windFrequency=<0 4 10>
    windOffset=<-450 -100 100>
   }
   skydive_FloatingDown {
    maxDownInterpSpeed=5
    maxGustDuration=8
    maxGustStrength=1
    maxTimeBetweenGusts=8
    maxUpInterpSpeed=5
    minDownInterpSpeed=2
    minGustDuration=0.2
    minGustStrength=0
    minTimeBetweenGusts=1
    minUpInterpSpeed=2
    windTurbulenceScalarA=0.5
    windTurbulenceScalarB=1
    windAmplitude=<50 20 35>
    windFrequency=<8 2 3>
    windOffset=<0 -300 200>
   }
   skydive_GoingUp {
    maxDownInterpSpeed=5
    maxGustDuration=8
    maxGustStrength=7
    maxTimeBetweenGusts=8
    maxUpInterpSpeed=5
    minDownInterpSpeed=2
    minGustDuration=0.2
    minGustStrength=3
    minTimeBetweenGusts=1
    minUpInterpSpeed=2
    windTurbulenceScalarA=0.1
    windTurbulenceScalarB=0.8
    windAmplitude=<0 0 0>
    windFrequency=<0 0 0>
    windOffset=<0 -500 300>
   }
   skydive_Parachuting {
    maxDownInterpSpeed=5
    maxGustDuration=8
    maxGustStrength=1
    maxTimeBetweenGusts=8
    maxUpInterpSpeed=5
    minDownInterpSpeed=2
    minGustDuration=5
    minGustStrength=0
    minTimeBetweenGusts=5
    minUpInterpSpeed=2
    windTurbulenceScalarA=0.2
    windTurbulenceScalarB=1
    windAmplitude=<25 50 75>
    windFrequency=<8 6 2>
    windOffset=<-150 -400 -150>
   }
   skydive_Surfing {
    maxDownInterpSpeed=5
    maxGustDuration=2
    maxGustStrength=7
    maxTimeBetweenGusts=7
    maxUpInterpSpeed=5
    minDownInterpSpeed=2
    minGustDuration=0.5
    minGustStrength=3
    minTimeBetweenGusts=0
    minUpInterpSpeed=2
    windTurbulenceScalarA=0.1
    windTurbulenceScalarB=1
    windAmplitude=<0 0 0>
    windFrequency=<0 0 0>
    windOffset=<0 0 0>
   }
   swimming {
    maxDownInterpSpeed=5
    maxGustDuration=2
    maxGustStrength=7
    maxTimeBetweenGusts=7
    maxUpInterpSpeed=5
    minDownInterpSpeed=2
    minGustDuration=0.5
    minGustStrength=3
    minTimeBetweenGusts=0
    minUpInterpSpeed=2
    windTurbulenceScalarA=0.1
    windTurbulenceScalarB=1
    windAmplitude=<10 25 20>
    windFrequency=<1.5 0.5 7.5>
    windOffset=<0 -150 50>
   }
   tactical_Sprint {
    maxDownInterpSpeed=5
    maxGustDuration=2
    maxGustStrength=7
    maxTimeBetweenGusts=7
    maxUpInterpSpeed=5
    minDownInterpSpeed=2
    minGustDuration=0.5
    minGustStrength=3
    minTimeBetweenGusts=0
    minUpInterpSpeed=2
    windTurbulenceScalarA=0.1
    windTurbulenceScalarB=1
    windAmplitude=<0 0 0>
    windFrequency=<0 0 0>
    windOffset=<0 0 0>
   }
   ziplining {
    maxDownInterpSpeed=5
    maxGustDuration=2
    maxGustStrength=7
    maxTimeBetweenGusts=7
    maxUpInterpSpeed=5
    minDownInterpSpeed=2
    minGustDuration=0.5
    minGustStrength=3
    minTimeBetweenGusts=0
    minUpInterpSpeed=2
    windTurbulenceScalarA=0.1
    windTurbulenceScalarB=1
    windAmplitude=<0 0 0>
    windFrequency=<0 0 0>
    windOffset=<0 0 0>
   }
  }
 }
]
