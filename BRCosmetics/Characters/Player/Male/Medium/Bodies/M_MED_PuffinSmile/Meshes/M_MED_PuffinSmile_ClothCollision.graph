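# Cloth collision physics asset for the male medium PuffinSmile cosmetic body, providing simple shapes for simulated parts to collide against.
# Capsules cover the pelvis and lower spine, and near-identical sideways capsules sit on each bone of the five-bone left and right dyn_belt chains.
# A large flat box on dyn_ground acts as a floor below that bone.
# Leg, foot and forearm bodies have no shapes, and there are no constraints.
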PhysicsAsset /BRCosmetics/Characters/Player/Male/Medium/Bodies/M_MED_PuffinSmile/Meshes/M_MED_PuffinSmile_ClothCollision
bodies {
 calf_l []
 calf_r []
 dyn_belt_l_01 [{shapeType=Capsule, length=11.8293, radius=8, shapeName=dyn_belt_r_01_capsule, center=<3.5545 0 6.6896>, rotation=<rot 90 0 0>}]
 dyn_belt_l_02 [{shapeType=Capsule, length=11.8293, radius=8, shapeName=dyn_belt_r_01_capsule, center=<3.5545 0 6.6896>, rotation=<rot 90 0 0>}]
 dyn_belt_l_03 [{shapeType=Capsule, length=11.8293, radius=8, shapeName=dyn_belt_r_01_capsule, center=<3.5545 0 6.6896>, rotation=<rot 90 0 0>}]
 dyn_belt_l_04 [{shapeType=Capsule, length=11.8293, radius=8, shapeName=dyn_belt_r_01_capsule, center=<3.5545 0 6.6896>, rotation=<rot 90 0 0>}]
 dyn_belt_l_05 [{shapeType=Capsule, length=11.8293, radius=8, shapeName=dyn_belt_r_01_capsule, center=<3.5545 0 6.6896>, rotation=<rot 90 0 0>}]
 dyn_belt_r_01 [{shapeType=Capsule, length=11.8293, radius=8, shapeName=dyn_belt_r_01_capsule, center=<3.5545 0 6.6896>, rotation=<rot 90 0 0>}]
 dyn_belt_r_02 [{shapeType=Capsule, length=11.8293, radius=8, shapeName=dyn_belt_r_01_capsule, center=<3.5545 0 6.6896>, rotation=<rot 90 0 0>}]
 dyn_belt_r_03 [{shapeType=Capsule, length=11.8293, radius=8, shapeName=dyn_belt_r_01_capsule, center=<3.5545 0 6.3984>, rotation=<rot 90 0 0>}]
 dyn_belt_r_04 [{shapeType=Capsule, length=11.8293, radius=8, shapeName=dyn_belt_r_01_capsule, center=<3.5545 0 6.3984>, rotation=<rot 90 0 0>}]
 dyn_belt_r_05 [{shapeType=Capsule, length=11.8293, radius=8, shapeName=dyn_belt_r_01_capsule, center=<3.5545 0 6.3984>, rotation=<rot 90 0 0>}]
 dyn_ground [{shapeType=Box, extentX=327.5, extentY=327.5, extentZ=59.22, shapeName=root_box, center=<0 0 -29.0385>}]
 foot_l []
 foot_r []
 lowerarm_l []
 lowerarm_r []
 pelvis [
  {shapeType=Capsule, length=8.5164, radius=10.7208, center=<-2.5281 -2.1225 0.0001>}
  {
   shapeType=Capsule
   length=14.3328
   radius=10.9706
   shapeName=pelvis_capsule
   center=<-2.7633 -1.7866 6.9039>
   rotation=<rot 88.5684 2.5631 2.7711>
  }
 ]
 spine_01 [{shapeType=Capsule, length=7.1527, radius=9.7082, center=<0.3095 -0.6696 0>}]
 spine_02 [{shapeType=Capsule, length=8.5692, radius=8.9596, center=<-1.0816 -0.436 -0.0002>}]
 thigh_l []
 thigh_r []
}
constraints []
